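# Unlit translucent Niagara sprite material for the FantasyPlant aura in Sparks cosmetics, sampling a plant aura spritesheet.
# Color comes from particle color, optionally blended with a Dynamic Parameter color through a texture channel mask, then optionally multiplied by texture RGB.
# An optional time-of-day tint multiplies that color before a shared emissive adjustment function.
# Particle and texture alpha form opacity, with depth fade skipped on mobile.
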Material /SparksCosmetics/Items/Auras/Assets/NS_Aura_FantasyPlant/M_Aura_FantasyPlant_Sprites
graph {
 properties {
  bAutomaticallySetUsageInEditor=false
  bUsedWithNiagaraSprites=true
  blendMode=BLEND_Translucent
  shadingModel=MSM_Unlit
  usedShadingModels=Unlit
 }
 nodes [
  n1 = Constant3Vector() {constant=<rgba 1 1 1 0>}
  n2 = CollectionParameter() {
   collection=&FortniteMaterialParameters
   group=ToD
   parameterName=FXTimeOfDayColorTint
  }
  n3 = ComponentMask(None=@n2) {b=true, g=true, r=true}
  n4 = ScalarParameter() {group=Color, parameterName=ToD Contribution, sortPriority=33}
  n5 = TextureSampleParameter2D(UVs=TextureCoordinate()) {parameterName=Main Tex, texture=&/T_PlantAura_Spritesheet_4x4}
  n6 = ParticleColor()
  n7 = LinearInterpolate(
   A=@n6.RGB
   B=LinearInterpolate(
    A=DynamicParameter() {defaultValue=<rgba 0 0 0 1>, paramNames [R,G,B,Param4]}.RGB
    B=@n6.RGB
    Alpha=StaticComponentMaskParameter(None=@n5.RGBA) {defaultR=true, group=Color, parameterName=Lerp Dynamic Color Channel Mask}
   )
   Alpha=ScalarParameter() {desc="0 = particle color only, 1  = use dynamic param color lerp", group=Color, parameterName=Use Dynamic Color Lerp}
  )
  n8 = Multiply(
   A=@n7
   B=StaticComponentMaskParameter(None=@n5.RGBA) {defaultB=true, defaultG=true, defaultR=true, group=Color, parameterName=RGB Color Channel Mask}
  )
  n9 = ScalarParameter() {defaultValue=1, desc="0 = use texture color X particle color, 1 = particle color", group=Color, parameterName=Use Texture Color}
  MaterialFunctionCall(Emissive=Multiply(A=LinearInterpolate(A=@n1, B=@n3, Alpha=@n4), B=LinearInterpolate(A=@n7, B=@n8, Alpha=@n9))) {materialFunction=&MF_FX_EmissiveAdjustments}
  n10 = Multiply(A=@n6.A, B=StaticComponentMaskParameter(None=@n5.RGBA) {defaultA=true, group=Opacity, parameterName=Alpha Channel Mask})
  n11 = DepthFade(Opacity=@n10, FadeDistance=ScalarParameter() {defaultValue=5, group=Opacity, parameterName=Depth Fade Distance})
  ShadingPathSwitch(Default=@n11, Deferred=@n11, Mobile=@n10) {desc=Migrated from FeatureLevelSwitch}
 ]
}
