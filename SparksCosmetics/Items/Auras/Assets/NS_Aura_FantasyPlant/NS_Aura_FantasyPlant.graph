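# Niagara system for the Sparks cosmetic FantasyPlant aura, which dresses a skeletal mesh in glows and leafy stems.
# A combined glow emitter pins four sprite groups to bones or sockets, adjusting for body size, front end and time of day.
# Leaf sprites read Stem_Ribbons particles to sit along each stem, with per-ribbon shape, rotation and color.
# Leaves also compute a smoothed camera-facing value, which two disabled debug emitters visualize.
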NiagaraSystem /SparksCosmetics/Items/Auras/Assets/NS_Aura_FantasyPlant/NS_Aura_FantasyPlant
dependencies {
 usedDataInterfaces [
  &NiagaraDataInterfaceColorCurve
  &NiagaraDataInterfaceArrayFloat
  &NiagaraDataInterfaceCurve
  &NiagaraDataInterfaceArrayInt32
  &NiagaraDataInterfaceArrayFloat2
  &NiagaraDataInterfaceSpriteRendererInfo
  &NiagaraDataInterfaceCamera
  &NiagaraDataInterfaceArrayFloat3
  &NiagaraDataInterfaceParticleRead
  &NiagaraDataInterfaceSkeletalMesh
 ]
 usedDynamicInputs []
 usedModules [
  &self:SystemSpawnScript.SystemScriptSource.SystemScriptGraph.NiagaraNodeAssignment_0.SetVariables_#1
  &self:SystemSpawnScript.SystemScriptSource.SystemScriptGraph.NiagaraNodeAssignment_3.SetVariables_#2
  &SystemState
  &self:SystemSpawnScript.SystemScriptSource.SystemScriptGraph.NiagaraNodeAssignment_2.SetVariables_#3
  &self:SystemSpawnScript.SystemScriptSource.SystemScriptGraph.NiagaraNodeAssignment_1.SetVariables_#4
 ]
}
emitters {
 DebugRotationLerp {
  bEnabled=false
  emitterUpdateScript [
   &EmitterState()
   &SpawnBurst_Instantaneous(Use Spawn Probability=false, UseLoopCountLimit=true)
  ]
  particleSpawnScript [&InitializeParticle(
    Color Mode=::NewEnumerator0
    Position Offset=<0 0 50>
    Position Offset Coordinate Space=::Local
    Sprite Size Mode=::NewEnumerator3
    Uniform Sprite Size=50
    UsePositionOffset=true
    Write Material Random=true
   )]
  particleUpdateScript [
   &ParticleState(Kill Particles When Lifetime Has Elapsed=false)
   &SkeletalMeshLocation(
    Bone / Socket Index=7
    Bone Sampling Mode=::NewEnumerator4
    MaskSpawnGroup=false
    Position Sampling=::NewEnumerator2
    Skeletal Mesh=$System.Skeletal Mesh
    Use Sampled Mesh Attribute as Ribbon ID=false
   )
   /SetVariables(Particles.CameraDirectionFloat:NiagaraFloat=/Game/Effects/Niagara/DynamicInputs/Math/FN_Fit(
     Clamp:NiagaraBool=false
     Value:NiagaraFloat=&DotProduct(
      Vector A=&ReturnCameraVector(Camera Vectors=::NewEnumerator1, Coordinate Space=::Local)
      Vector B=&ReturnMeshOrientationAxis(Mesh Orientation Axis=::YAxis, Mesh Orientation Quat=$Particles.MeshOrientation)
     )
     SourceMin:NiagaraFloat=-1
     SourceMax:NiagaraFloat=1
     DestMin:NiagaraFloat=0
     DestMax:NiagaraFloat=1
    )) {isSetParameters=true, module=SetVariables_#5}
   &InterpolateOverTime(Target Float Value:NiagaraFloat=$Particles.CameraDirectionFloat)
   &DynamicMaterialParameters(
    Index 0 Param 1=$Particles.InterpolateOverTime.InterpolateOverTime_Float.Moving Average
    Module.Index0Float0Write=true
    Module.Index0Float1Write=true
    Module.Index0Float2Write=true
    Module.Index0Float3Write=true
   )
   &CameraOffset(Camera Offset Amount=75)
  ]
  rendererClasses [&NiagaraSpriteRendererProperties, &NiagaraMeshRendererProperties]
  renderers [
   {rendererClass=&NiagaraSpriteRendererProperties}
   {rendererClass=&NiagaraMeshRendererProperties, rendererIndex=1}
  ]
 }
 DebugRotationLerp002 {
  bEnabled=false
  emitterUpdateScript [
   &EmitterState()
   &SpawnBurst_Instantaneous(Use Spawn Probability=false, UseLoopCountLimit=true)
  ]
  particleSpawnScript [&InitializeParticle(
    Color Mode=::NewEnumerator0
    Position Offset=<0 0 50>
    Position Offset Coordinate Space=::Local
    Sprite Size Mode=::NewEnumerator3
    Uniform Sprite Size=50
    UsePositionOffset=true
    Write Material Random=true
   )]
  particleUpdateScript [
   &ParticleState(Kill Particles When Lifetime Has Elapsed=false)
   &SkeletalMeshLocation(
    Bone / Socket Index=7
    Bone Sampling Mode=::NewEnumerator4
    MaskSpawnGroup=false
    Position Sampling=::NewEnumerator2
    Skeletal Mesh=$System.Skeletal Mesh
    Use Sampled Mesh Attribute as Ribbon ID=false
   )
   /SetVariables(Particles.CameraDirectionFloat:NiagaraFloat=&FN_Fit(
     Clamp:NiagaraBool=false
     Value:NiagaraFloat=&DotProduct(
      Vector A=&ReturnCameraVector(Camera Vectors=::NewEnumerator1, Coordinate Space=::Local)
      Vector B=&ReturnMeshOrientationAxis(Mesh Orientation Axis=::YAxis, Mesh Orientation Quat=$Particles.MeshOrientation)
     )
     SourceMin:NiagaraFloat=-1
     SourceMax:NiagaraFloat=1
     DestMin:NiagaraFloat=0
     DestMax:NiagaraFloat=1
    )) {isSetParameters=true, module=SetVariables_#5}
   &InterpolateOverTime(Rate Of Change=3, Target Float Value:NiagaraFloat=$Particles.CameraDirectionFloat)
   &DynamicMaterialParameters(
    Index 0 Param 1=$Particles.InterpolateOverTime.InterpolateOverTime_Float.Moving Average
    Module.Index0Float0Write=true
    Module.Index0Float1Write=true
    Module.Index0Float2Write=true
    Module.Index0Float3Write=true
   )
   &CameraOffset(Camera Offset Amount=75)
  ]
  rendererClasses [&NiagaraSpriteRendererProperties, &NiagaraMeshRendererProperties]
  renderers [
   {rendererClass=&NiagaraSpriteRendererProperties}
   {rendererClass=&NiagaraMeshRendererProperties, rendererIndex=1}
  ]
 }
 Decal_Glows_Combined {
  emitterUpdateScript [
   &EmitterState()
   &SpawnBurst_Instantaneous(Spawn Count=5, Use Spawn Probability=false, UseLoopCountLimit=true)
   &SpawnBurst_Instantaneous(Spawn Group=1, Use Spawn Probability=false, UseLoopCountLimit=true) {module=SpawnBurst_Instantaneous001}
   &SpawnBurst_Instantaneous(Spawn Count=8, Spawn Group=2, Use Spawn Probability=false, UseLoopCountLimit=true) {module=SpawnBurst_Instantaneous002}
   &SpawnBurst_Instantaneous(Spawn Count=5, Spawn Group=3, Use Spawn Probability=false, UseLoopCountLimit=true) {module=SpawnBurst_Instantaneous003}
  ]
  particleSpawnScript [
   /SetVariables(Particles.VisibilityTag=&SpawnGroup()) {isSetParameters=true}
   &PartitionParticles(Number Of Partitions=5)
   &PartitionParticles(
    Normalized Index In Partition Includes One=true
    Number Of Partitions=4
    Particles Per Partition=2
    Partition Mode:ENiagara_PartitionMode_v2=::NewEnumerator1
   ) {module=PartitionParticles001}
   &PartitionParticles(Number Of Partitions=5) {module=PartitionParticles002}
   &InitializeParticle(
    Lifetime=3
    Sprite Size Mode=::NewEnumerator3
    Uniform Sprite Size=&MergeHelper_FloatSwitch(
     0=&Multiply_Float(
      A=&SelectFloatFromArray(
       Direct Array Index=$Particles.PartitionParticles.Partition
       Float Selection Array=DataInterface() {propertyValues {FloatData [35,35,35,30,30]}}
      )
      B=&MakeCustomFloatFromBool(Bool=$User.IsLargeBody, True Float=1.2, False Float=1)
     )
     1=0
     2=&MakeCustomFloatFromBool(Bool=$User.IsLargeBody, True Float=38, False Float=33)
     3=&MakeFloatFromVector2d(Vector 2D=&Multiply_Vector2D_ByFloat(
       Float=&Multiply_Float(
        A=&SelectFloatFromArray(
         Direct Array Index=$Particles.PartitionParticles002.Partition
         Float Selection Array=DataInterface() {propertyValues {FloatData [75,55,55,50,50]}}
        )
        B=&MakeCustomFloatFromBool(Bool=$User.IsLargeBody, True Float=0.9, False Float=0.7)
       )
       Vector2D=<0.9 0.9>
      ))
     Index=&SpawnGroup()
     Num=::NewEnumerator2
    )
    UsePositionOffset=false
    Write Material Random=true
   )
   &CameraOffset(Camera Offset Amount=&MergeHelper_FloatSwitch(
     0=&SelectFloatFromArray(
      Direct Array Index=$Particles.PartitionParticles.Partition
      Float Selection Array=DataInterface() {propertyValues {FloatData [6,10,10,10,10,10,10]}}
     )
     1=0
     2=&MergeHelper_FloatSwitch(
      0=&FloatFromCurve(
       FloatCurve=DataInterface() {propertyValues {Curve {defaultValue=1.2, keys [{tangentMode=RCTM_User, time=0, value=2}, {tangentMode=RCTM_User, time=1, value=1.3}]}}}
       CurveIndex=$Particles.PartitionParticles001.Normalized Index In Partition
       Scale Curve=8
      )
      1=&FloatFromCurve(
       FloatCurve=DataInterface() {propertyValues {Curve {defaultValue=1.2, keys [{tangentMode=RCTM_User, time=0, value=2}, {tangentMode=RCTM_User, time=1, value=1.3}]}}}
       CurveIndex=$Particles.PartitionParticles001.Normalized Index In Partition
       Scale Curve=8
      )
      2=&FloatFromCurve(
       FloatCurve=DataInterface() {propertyValues {Curve {defaultValue=1, keys [{tangentMode=RCTM_User, time=0, value=1.2}, {tangentMode=RCTM_User, time=1, value=2}]}}}
       CurveIndex=$Particles.PartitionParticles001.Normalized Index In Partition
       Scale Curve=8
      )
      3=&FloatFromCurve(
       FloatCurve=DataInterface() {propertyValues {Curve {defaultValue=1, keys [{tangentMode=RCTM_User, time=0, value=1.2}, {tangentMode=RCTM_User, time=1, value=2}]}}}
       CurveIndex=$Particles.PartitionParticles001.Normalized Index In Partition
       Scale Curve=8
      )
      Index=$Particles.PartitionParticles001.Partition
      Num=::NewEnumerator2
     )
     3=-10
     Index=&SpawnGroup()
     Num=::NewEnumerator2
    ))
   /SetVariables(Particles.GlowCameraOffset=-30) {isSetParameters=true}
   /SetVariables(Particles.GlowScale:Vector2f=&Multiply_Vector2D(
     A=<65 65>
     B=&MakeCustomVector2DFromBool(Bool=$User.IsLargeBody, True Vector2D=<1.4 1.4>, False Vector2D=<1 1>)
    )) {isSetParameters=true}
   /SetVariables(Particles.GlowColor=&Multiply_LinearColorByFloat(Linear Color=<rgba 1 1 1 1>, Float=15)) {isSetParameters=true}
   /SetVariables(Particles.Decal Dynamic Param:Vector4f=&MakeVector4(X=5, Y=0.1, Z=0, W=0)) {isSetParameters=true}
  ]
  particleUpdateScript [
   &ParticleState(Kill Particles When Lifetime Has Elapsed=false, Let Infinitely Lived Particles Die When Emitter Deactivates=true)
   &SkeletalMeshLocation(
    Bone / Socket Index=&MergeHelper_IntSwitch(
     0=&SelectIntFromArray(Direct Array Index=$Particles.PartitionParticles.Partition, Int Selection Array=DataInterface() {propertyValues {IntData [0,1,2,3,4]}})
     1=7
     2=&SelectIntFromArray(Direct Array Index=$Particles.PartitionParticles001.Partition, Int Selection Array=DataInterface() {propertyValues {IntData [1,2,7,7]}})
     3=&SelectIntFromArray(Direct Array Index=$Particles.PartitionParticles002.Partition, Int Selection Array=DataInterface() {propertyValues {IntData [0,1,2,1,2]}})
     4=?
     5=?
     Index=&SpawnGroup()
     Num=::NewEnumerator2
    )
    Bone Sampling Mode=::NewEnumerator4
    MaskSpawnGroup=false
    Orientation Sampling=::NewEnumerator1
    Sampled Position Offset=&MergeHelper_VectorSwitch(
     0=&SelectVectorFromArray(
      Direct Array Index=$Particles.PartitionParticles.Partition
      Vector Selection Array=DataInterface() {propertyValues {FloatData [<-5 0 0>, <0 0 0>, <0 0 0>, <0 0 0>, <0 0 0>, <0 0 0>, <0 0 0>]}}
     )
     1=<5 0 0>
     2=&MergeHelper_VectorSwitch(
      0=&MakeVector(
       X=&FloatFromCurve(
        FloatCurve=DataInterface() {propertyValues {Curve {defaultValue=1, keys [{tangentMode=RCTM_User, time=0, value=0}, {tangentMode=RCTM_User, time=1, value=1}]}}}
        CurveIndex=$Particles.PartitionParticles001.Normalized Index In Partition
        Scale Curve=&MakeCustomFloatFromBool(Bool=$User.IsLargeBody, True Float=9, False Float=7)
       )
       Y=&FloatFromCurve(
        FloatCurve=DataInterface() {propertyValues {Curve {keys [<at 0 1>,<at 1 1>]}}}
        CurveIndex=$Particles.PartitionParticles001.Normalized Index In Partition
        Scale Curve=-2
       )
       Z=&FloatFromCurve(
        FloatCurve=DataInterface() {propertyValues {Curve {keys [<at 0 1>,<at 1 1>]}}}
        CurveIndex=$Particles.PartitionParticles001.Normalized Index In Partition
        Scale Curve=2
       )
      )
      1=&MakeVector(
       X=&FloatFromCurve(
        FloatCurve=DataInterface() {propertyValues {Curve {defaultValue=1, keys [{tangentMode=RCTM_User, time=0, value=0}, {tangentMode=RCTM_User, time=1, value=1}]}}}
        CurveIndex=$Particles.PartitionParticles001.Normalized Index In Partition
        Scale Curve=&MakeCustomFloatFromBool(Bool=$User.IsLargeBody, True Float=-9, False Float=-7)
       )
       Y=&FloatFromCurve(
        FloatCurve=DataInterface() {propertyValues {Curve {keys [<at 0 1>,<at 1 1>]}}}
        CurveIndex=$Particles.PartitionParticles001.Normalized Index In Partition
        Scale Curve=2
       )
       Z=&FloatFromCurve(
        FloatCurve=DataInterface() {propertyValues {Curve {keys [<at 0 1>,<at 1 1>]}}}
        CurveIndex=$Particles.PartitionParticles001.Normalized Index In Partition
        Scale Curve=-2
       )
      )
      2=&MakeVector(
       X=&FloatFromCurve(
        FloatCurve=DataInterface() {propertyValues {Curve {defaultValue=1, keys [{tangentMode=RCTM_User, time=0, value=1}, {tangentMode=RCTM_User, time=1, value=1}]}}}
        CurveIndex=$Particles.PartitionParticles001.Normalized Index In Partition
        Scale Curve=5
       )
       Y=&FloatFromCurve(
        FloatCurve=DataInterface() {propertyValues {Curve {keys [<at 0 0>,<at 1 0>]}}}
        CurveIndex=$Particles.PartitionParticles001.Normalized Index In Partition
       )
       Z=&FloatFromCurve(
        FloatCurve=DataInterface() {propertyValues {Curve {keys [<at 0 0.6>,<at 1 1>]}}}
        CurveIndex=$Particles.PartitionParticles001.Normalized Index In Partition
        Scale Curve=&MakeCustomFloatFromBool(Bool=$User.IsLargeBody, True Float=18, False Float=12)
       )
      )
      3=&MakeVector(
       X=&FloatFromCurve(
        FloatCurve=DataInterface() {propertyValues {Curve {defaultValue=1, keys [{tangentMode=RCTM_User, time=0, value=1}, {tangentMode=RCTM_User, time=1, value=1}]}}}
        CurveIndex=$Particles.PartitionParticles001.Normalized Index In Partition
        Scale Curve=5
       )
       Y=&FloatFromCurve(
        FloatCurve=DataInterface() {propertyValues {Curve {keys [<at 0 0>,<at 1 0>]}}}
        CurveIndex=$Particles.PartitionParticles001.Normalized Index In Partition
       )
       Z=&FloatFromCurve(
        FloatCurve=DataInterface() {propertyValues {Curve {keys [<at 0 -0.6>,<at 1 -1>]}}}
        CurveIndex=$Particles.PartitionParticles001.Normalized Index In Partition
        Scale Curve=&MakeCustomFloatFromBool(Bool=$User.IsLargeBody, True Float=18, False Float=12)
       )
      )
      Index=$Particles.PartitionParticles001.Partition
      Num=::NewEnumerator2
     )
     3=&SelectVectorFromArray(
      Direct Array Index=$Particles.PartitionParticles002.Partition
      Vector Selection Array=DataInterface() {propertyValues {FloatData [<0 -5 0>, <6 0 0>, <-6 0 0>, <18 0 0>, <-18 0 0>]}}
     )
     Index=&SpawnGroup()
     Num=::NewEnumerator2
    )
    Skeletal Mesh=$System.Skeletal Mesh
   )
   &SolveForcesAndVelocity(Clamp Velocity=false, Limit Acceleration=false) {enabled=false}
   /SetVariables(
    Particles.GlowColor=&MakeLinearColorFromVectorAndFloat(
     "Vector (RGB)"=&Multiply_Vector(A=$System.Color RGB, B=&MakeVectorFromLinearColorRGB(Color=$Particles.Initial.GlowColor))
     "Float (Alpha)"=&Multiply_Float(
      A=&FloatFromCurve(
       FloatCurve=DataInterface() {propertyValues {Curve {keys [
           <at 0.1 0.8 RCIM_Cubic>
           {arriveTangentWeight=0.1077, interpMode=RCIM_Cubic, leaveTangentWeight=0.0718, tangentMode=RCTM_User, time=0.4, value=1}
           {arriveTangentWeight=0.0718, interpMode=RCIM_Cubic, leaveTangentWeight=0.1077, tangentMode=RCTM_User, time=0.6, value=1}
           <at 0.9 0.8 RCIM_Cubic>
          ]}}}
       CurveIndex=Linked() {linkedVariable {name=NPC.FortniteNPC.TimeOfDay_0-1, type=NiagaraFloat}}
      )
      B=&MakeCustomFloatFromBool(Bool=$User.IsFrontEnd, True Float=0.4, False Float=1)
     )
    )
    Particles.GlowScale:Vector2f=&Multiply_Vector2D(A=$Particles.Initial.GlowScale, B=&Vector2DFromFloat(Value=1))
   ) {isSetParameters=true}
   /SetVariables(Particles.PositionOffset:NiagaraPosition=&AddVectorToPosition(
     Position=$Particles.Position
     Vector=&MergeHelper_VectorSwitch(
      0=<0 0 0>
      1=<0 0 0>
      2=&MakeVector(
       X=0
       Y=0
       Z=&FloatFromCurve(
        FloatCurve=DataInterface() {propertyValues {Curve {keys [<at 0 5 RCIM_Cubic>,<at 1 2 RCIM_Cubic>]}}}
        CurveIndex=$Particles.PartitionParticles001.Normalized Index In Partition
        Scale Curve=&MakeCustomFloatFromBool(Bool=$User.IsLargeBody, True Float=2.5, False Float=1)
       )
      )
      3=&MakeVector(
       X=0
       Y=0
       Z=&FloatFromCurve(
        FloatCurve=DataInterface() {propertyValues {Curve {keys [<at 0 5 RCIM_Cubic>,<at 1 2 RCIM_Cubic>]}}}
        CurveIndex=$Particles.PartitionParticles001.Normalized Index In Partition
        Scale Curve=&MakeCustomFloatFromBool(Bool=$User.IsLargeBody, True Float=2.5, False Float=1)
       )
      )
      Index=$Particles.PartitionParticles001.Partition
      Num=::NewEnumerator2
     )
    )) {isSetParameters=true}
   &ScaleSpriteSize(
    Scale Factor=&MergeHelper_Vector2Switch(
     0=<1 1>
     1=<1 1>
     2=<1 1>
     3=&Multiply_Vector2D_ByFloat(
      Float=&MakeCustomFloatFromBool(Bool=$User.IsLargeBody, True Float=2.2, False Float=1.8)
      Vector2D=<0.7 1>
     )
     4=?
     Index=&SpawnGroup()
     Num=::NewEnumerator2
    )
    Scale Sprite Size Mode=::NewEnumerator2
   ) {module=ScaleSpriteSize001}
   &ScaleColor(
    Scale Alpha=&MergeHelper_FloatSwitch(
     0=&FloatFromCurve(
      FloatCurve=DataInterface() {propertyValues {Curve {keys [
          {arriveTangent=-0.0686, interpMode=RCIM_Cubic, leaveTangent=-0.0686, tangentMode=RCTM_User, time=0, value=0.9}
          {
           arriveTangent=0.0901
           arriveTangentWeight=0.0676
           interpMode=RCIM_Cubic
           leaveTangent=0.0901
           leaveTangentWeight=0.0672
           tangentMode=RCTM_User
           time=0.1964
           value=0.9848
          }
          {arriveTangentWeight=0.0651, interpMode=RCIM_Cubic, leaveTangentWeight=0.0357, tangentMode=RCTM_User, time=0.3917, value=1}
          {arriveTangentWeight=0.0357, interpMode=RCIM_Cubic, leaveTangentWeight=0.0349, tangentMode=RCTM_User, time=0.4987, value=1}
          {arriveTangentWeight=0.0349, interpMode=RCIM_Cubic, leaveTangentWeight=0.0655, tangentMode=RCTM_User, time=0.6031, value=1}
          {
           arriveTangent=-0.185
           arriveTangentWeight=0.0674
           interpMode=RCIM_Cubic
           leaveTangent=-0.185
           leaveTangentWeight=0.069
           tangentMode=RCTM_User
           time=0.7993
           value=0.9848
          }
          {arriveTangent=-0.0686, interpMode=RCIM_Cubic, leaveTangent=-0.0686, tangentMode=RCTM_User, time=1, value=0.9}
         ]}}}
      CurveIndex=Linked() {linkedVariable {name=NPC.FortniteNPC.TimeOfDay_0-1, type=NiagaraFloat}}
      Scale Curve=&MakeCustomFloatFromBool(Bool=$User.IsFrontEnd, True Float=0.6, False Float=1)
     )
     1=&FloatFromCurve(
      FloatCurve=DataInterface() {propertyValues {Curve {keys [
          {arriveTangent=-0.0686, interpMode=RCIM_Cubic, leaveTangent=-0.0686, tangentMode=RCTM_User, time=0, value=0.9}
          {
           arriveTangent=0.0901
           arriveTangentWeight=0.0676
           interpMode=RCIM_Cubic
           leaveTangent=0.0901
           leaveTangentWeight=0.0672
           tangentMode=RCTM_User
           time=0.1964
           value=0.9848
          }
          {arriveTangentWeight=0.0651, interpMode=RCIM_Cubic, leaveTangentWeight=0.0357, tangentMode=RCTM_User, time=0.3917, value=1}
          {arriveTangentWeight=0.0357, interpMode=RCIM_Cubic, leaveTangentWeight=0.0349, tangentMode=RCTM_User, time=0.4987, value=1}
          {arriveTangentWeight=0.0349, interpMode=RCIM_Cubic, leaveTangentWeight=0.0655, tangentMode=RCTM_User, time=0.6031, value=1}
          {
           arriveTangent=-0.185
           arriveTangentWeight=0.0674
           interpMode=RCIM_Cubic
           leaveTangent=-0.185
           leaveTangentWeight=0.069
           tangentMode=RCTM_User
           time=0.7993
           value=0.9848
          }
          {arriveTangent=-0.0686, interpMode=RCIM_Cubic, leaveTangent=-0.0686, tangentMode=RCTM_User, time=1, value=0.9}
         ]}}}
      CurveIndex=Linked() {linkedVariable {name=NPC.FortniteNPC.TimeOfDay_0-1, type=NiagaraFloat}}
      Scale Curve=&MakeCustomFloatFromBool(Bool=$User.IsFrontEnd, True Float=0.6, False Float=1)
     )
     2=&FloatFromCurve(
      FloatCurve=DataInterface() {propertyValues {Curve {keys [
          {arriveTangent=-0.0686, interpMode=RCIM_Cubic, leaveTangent=-0.0686, tangentMode=RCTM_User, time=0, value=0.9}
          {
           arriveTangent=0.0901
           arriveTangentWeight=0.0676
           interpMode=RCIM_Cubic
           leaveTangent=0.0901
           leaveTangentWeight=0.0672
           tangentMode=RCTM_User
           time=0.1964
           value=0.9848
          }
          {arriveTangentWeight=0.0651, interpMode=RCIM_Cubic, leaveTangentWeight=0.0357, tangentMode=RCTM_User, time=0.3917, value=1}
          {arriveTangentWeight=0.0357, interpMode=RCIM_Cubic, leaveTangentWeight=0.0349, tangentMode=RCTM_User, time=0.4987, value=1}
          {arriveTangentWeight=0.0349, interpMode=RCIM_Cubic, leaveTangentWeight=0.0655, tangentMode=RCTM_User, time=0.6031, value=1}
          {
           arriveTangent=-0.185
           arriveTangentWeight=0.0674
           interpMode=RCIM_Cubic
           leaveTangent=-0.185
           leaveTangentWeight=0.069
           tangentMode=RCTM_User
           time=0.7993
           value=0.9848
          }
          {arriveTangent=-0.0686, interpMode=RCIM_Cubic, leaveTangent=-0.0686, tangentMode=RCTM_User, time=1, value=0.9}
         ]}}}
      CurveIndex=Linked() {linkedVariable {name=NPC.FortniteNPC.TimeOfDay_0-1, type=NiagaraFloat}}
      Scale Curve=&MakeCustomFloatFromBool(Bool=$User.IsFrontEnd, True Float=0.6, False Float=0.9)
     )
     3=&Multiply_Float(
      A=&MergeHelper_FloatSwitch(0=1, 1=1, 2=1, 3=0.75, 4=0.75, Index=$Particles.PartitionParticles002.Partition, Num=::NewEnumerator3)
      B=&FloatFromCurve(
       FloatCurve=DataInterface() {propertyValues {Curve {keys [
           <at 0 0.95 RCIM_Cubic>
           {arriveTangent=0.383, interpMode=RCIM_Cubic, leaveTangent=0.383, time=0.1964, value=1}
           {arriveTangentWeight=0.0686, interpMode=RCIM_Cubic, leaveTangentWeight=0.0376, tangentMode=RCTM_User, time=0.3917, value=1.1}
           <at 0.4987 1.1 RCIM_Cubic>
           {arriveTangentWeight=0.0367, interpMode=RCIM_Cubic, leaveTangentWeight=0.0689, tangentMode=RCTM_User, time=0.6031, value=1.1}
           {arriveTangent=-0.3779, interpMode=RCIM_Cubic, leaveTangent=-0.3779, time=0.7993, value=1}
           <at 1 0.95 RCIM_Cubic>
          ]}}}
       CurveIndex=Linked() {linkedVariable {name=NPC.FortniteNPC.TimeOfDay_0-1, type=NiagaraFloat}}
       Scale Curve=&MakeCustomFloatFromBool(Bool=$User.IsFrontEnd, True Float=0.5, False Float=0.75)
      )
     )
     Index=&SpawnGroup()
     Num=::NewEnumerator2
    )
    Scale RGB=&MergeHelper_VectorSwitch(
     0=&Multiply_VectorByFloat(Vector=$System.Color RGB, Float=5)
     1=<0 0 0>
     2=&Multiply_VectorByFloat(
      Vector=$System.Color RGB
      Float=&MakeCustomFloatFromBool(Bool=$User.IsFrontEnd, True Float=3.25, False Float=2.5)
     )
     3=&Multiply_VectorByFloat(
      Vector=$System.Color RGB
      Float=&MakeCustomFloatFromBool(Bool=$User.IsFrontEnd, True Float=2, False Float=1.5)
     )
     Index=&SpawnGroup()
     Num=::NewEnumerator2
    )
    ScaleA=true
    ScaleRGB=true
   )
  ]
  rendererClasses [&NiagaraSpriteRendererProperties]
  renderers [
   {rendererClass=&NiagaraSpriteRendererProperties}
   {rendererClass=&NiagaraSpriteRendererProperties, rendererIndex=1}
   {rendererClass=&NiagaraSpriteRendererProperties, rendererIndex=2}
   {rendererClass=&NiagaraSpriteRendererProperties, rendererIndex=3}
  ]
 }
 Leaf_Sprites001 {
  emitterUpdateScript [
   &EmitterState()
   /SetVariables(
    Emitter.Leaf Scale:NiagaraFloat=&MakeCustomFloatFromBool(Bool=$User.IsLargeBody, True Float=7.5, False Float=6.75)
    Emitter.LeavesPerRibbon=4
   ) {isSetParameters=true}
   &SpawnBurst_Instantaneous(
    Spawn Count=&Multiply_Int(A=$Emitter.LeavesPerRibbon, B=$System.NumberOfRibbons)
    Use Spawn Probability=false
    UseLoopCountLimit=true
   )
  ]
  particleSpawnScript [
   &PartitionParticles(
    Normalized Index In Partition Includes One=true
    Number Of Partitions=$System.NumberOfRibbons
    Particles Per Partition:NiagaraInt32=$Emitter.LeavesPerRibbon
    Partition Mode:ENiagara_PartitionMode_v2=::NewEnumerator1
   )
   /SetVariables(Particles.LeafPicker:NiagaraInt32=&MergeHelper_IntSwitch(
     0=&SelectIntFromArray(
      Direct Array Index=$Particles.PartitionParticles.Index In Partition
      Int Selection Array=DataInterface() {propertyValues {IntData [0,1,7,6]}}
     )
     1=&SelectIntFromArray(
      Direct Array Index=$Particles.PartitionParticles.Index In Partition
      Int Selection Array=DataInterface() {propertyValues {IntData [1,0,6,8]}}
     )
     2=&SelectIntFromArray(
      Direct Array Index=$Particles.PartitionParticles.Index In Partition
      Int Selection Array=DataInterface() {propertyValues {IntData [5,11,3,2]}}
     )
     3=&SelectIntFromArray(
      Direct Array Index=$Particles.PartitionParticles.Index In Partition
      Int Selection Array=DataInterface() {propertyValues {IntData [4,10,2,10]}}
     )
     4=&SelectIntFromArray(
      Direct Array Index=$Particles.PartitionParticles.Index In Partition
      Int Selection Array=DataInterface() {propertyValues {IntData [1,7,6,0]}}
     )
     5=&SelectIntFromArray(
      Direct Array Index=$Particles.PartitionParticles.Index In Partition
      Int Selection Array=DataInterface() {propertyValues {IntData [2,3,11,3]}}
     )
     Index=$Particles.PartitionParticles.Partition
     Num=::NewEnumerator4
    )) {isSetParameters=true}
   &InitializeParticle(
    Sprite Size Mode=::NewEnumerator3
    Sprite UV Mode=::NewEnumerator4
    Uniform Sprite Size=&MergeHelper_FloatSwitch(
     0=&FloatFromCurve(
      FloatCurve=DataInterface() {propertyValues {Curve {keys [<at 0 0.8093>,<at 0.25 0.7>,<at 0.6384 0.7959>,<at 0.9981 0.7468>]}}}
      CurveIndex=$Particles.PartitionParticles.Normalized Index In Partition
      Scale Curve=$Emitter.Leaf Scale
     )
     1=&FloatFromCurve(
      FloatCurve=DataInterface() {propertyValues {Curve {keys [<at 0 0.9>,<at 0.25 0.7>,<at 0.7019 0.8835>,<at 1 0.8182>]}}}
      CurveIndex=$Particles.PartitionParticles.Normalized Index In Partition
      Scale Curve=$Emitter.Leaf Scale
     )
     2=&FloatFromCurve(
      FloatCurve=DataInterface() {propertyValues {Curve {keys [<at 0 0.9>,<at 0.25 0.8>,<at 0.7019 0.8568>,<at 1 1.0118>]}}}
      CurveIndex=$Particles.PartitionParticles.Normalized Index In Partition
      Scale Curve=$Emitter.Leaf Scale
     )
     3=&FloatFromCurve(
      FloatCurve=DataInterface() {propertyValues {Curve {keys [<at 0 1.1>,<at 0.25 0.9524>,<at 0.75 0.9012>,<at 1 1>]}}}
      CurveIndex=$Particles.PartitionParticles.Normalized Index In Partition
      Scale Curve=$Emitter.Leaf Scale
     )
     4=&FloatFromCurve(
      FloatCurve=DataInterface() {propertyValues {Curve {keys [<at 0 0.8524>,<at 0.25 0.8518>,<at 0.6971 0.993>,<at 1 0.8059>]}}}
      CurveIndex=$Particles.PartitionParticles.Normalized Index In Partition
      Scale Curve=$Emitter.Leaf Scale
     )
     5=&FloatFromCurve(
      FloatCurve=DataInterface() {propertyValues {Curve {keys [<at 0.0171 0.91>,<at 0.25 0.6794>,<at 0.75 0.75>,<at 1.018 0.721>]}}}
      CurveIndex=$Particles.PartitionParticles.Normalized Index In Partition
      Scale Curve=$Emitter.Leaf Scale
     )
     Index=$Particles.PartitionParticles.Partition
     Num=::NewEnumerator4
    )
    UsePositionOffset=false
    Write Material Random=true
   )
   &SubUVAnimation(
    Direct Frame Index:NiagaraFloat=&SelectFloatFromArray(Direct Array Index=$Particles.LeafPicker, Float Selection Array=DataInterface() {propertyValues {FloatData [1,1,2,2,3,3,4,4,5,5,6,6]}})
    "Sprite Renderer "=DataInterface() {propertyValues {SpriteRenderer=&self:Leaf_Sprites001_0.NiagaraSpriteRendererProperties_0}}
    SubUV Animation Mode:ENiagara_SubUVLookupModeV2=::NewEnumerator5
   )
   /SetVariables(
    Particles.InitialRotation A:NiagaraFloat=&MergeHelper_FloatSwitch(
     0=&SelectFloatFromArray(
      Direct Array Index=$Particles.PartitionParticles.Index In Partition
      Float Selection Array=DataInterface() {propertyValues {FloatData [-25,-30,7,-50]}}
     )
     1=&SelectFloatFromArray(
      Direct Array Index=$Particles.PartitionParticles.Index In Partition
      Float Selection Array=DataInterface() {propertyValues {FloatData [0,0,-10,-15]}}
     )
     2=&SelectFloatFromArray(
      Direct Array Index=$Particles.PartitionParticles.Index In Partition
      Float Selection Array=DataInterface() {propertyValues {FloatData [8,-4,-25,0]}}
     )
     3=&SelectFloatFromArray(
      Direct Array Index=$Particles.PartitionParticles.Index In Partition
      Float Selection Array=DataInterface() {propertyValues {FloatData [-5,0,9.7983,0]}}
     )
     4=&SelectFloatFromArray(
      Direct Array Index=$Particles.PartitionParticles.Index In Partition
      Float Selection Array=DataInterface() {propertyValues {FloatData [25,20,0.8,13]}}
     )
     5=&SelectFloatFromArray(
      Direct Array Index=$Particles.PartitionParticles.Index In Partition
      Float Selection Array=DataInterface() {propertyValues {FloatData [3,9,30,10]}}
     )
     Index=$Particles.PartitionParticles.Partition
     Num=::NewEnumerator4
    )
    Particles.InitialRotationB:NiagaraFloat=&MergeHelper_FloatSwitch(
     0=&SelectFloatFromArray(
      Direct Array Index=$Particles.PartitionParticles.Index In Partition
      Float Selection Array=DataInterface() {propertyValues {FloatData [-2,8,30,10]}}
     )
     1=&SelectFloatFromArray(
      Direct Array Index=$Particles.PartitionParticles.Index In Partition
      Float Selection Array=DataInterface() {propertyValues {FloatData [0,0,-2.5,20]}}
     )
     2=&SelectFloatFromArray(
      Direct Array Index=$Particles.PartitionParticles.Index In Partition
      Float Selection Array=DataInterface() {propertyValues {FloatData [37,13,-1.5,0]}}
     )
     3=&SelectFloatFromArray(
      Direct Array Index=$Particles.PartitionParticles.Index In Partition
      Float Selection Array=DataInterface() {propertyValues {FloatData [-48.0674,0,-5.8304,-11.0737]}}
     )
     4=&SelectFloatFromArray(
      Direct Array Index=$Particles.PartitionParticles.Index In Partition
      Float Selection Array=DataInterface() {propertyValues {FloatData [25,-4.3339,-24.7418,-17.5646]}}
     )
     5=&SelectFloatFromArray(
      Direct Array Index=$Particles.PartitionParticles.Index In Partition
      Float Selection Array=DataInterface() {propertyValues {FloatData [-50,-45,-18,-56]}}
     )
     Index=$Particles.PartitionParticles.Partition
     Num=::NewEnumerator4
    )
    Particles.PivotOffsetPicker:Vector2f=&SelectVector2DFromArray(
     Direct Array Index=$Particles.LeafPicker
     Vector 2D Selection Array=DataInterface() {propertyValues {FloatData [
        <0.12 0.88>
        <0.88 0.88>
        <0.06 0.88>
        <0.94 0.88>
        <0.1 0.7>
        <0.9 0.7>
        <0.845 0.85>
        <0.155 0.85>
        <0.2 0.85>
        <0.8 0.85>
        <0.88 0.85>
        <0.12 0.85>
       ]}}
    )
    Particles.UVScalePicker:Vector2f=&SelectVector2DFromArray(
     Direct Array Index=$Particles.LeafPicker
     Vector 2D Selection Array=DataInterface() {propertyValues {FloatData [<1 1>, <-1 1>, <1 1>, <-1 1>, <1 1>, <-1 1>, <1 1>, <-1 1>, <1 1>, <-1 1>, <1 1>, <-1 1>]}}
    )
   ) {isSetParameters=true}
  ]
  particleUpdateScript [
   &ParticleState(Kill Particles When Lifetime Has Elapsed=false)
   &SkeletalMeshLocation(
    Bone / Socket Index=7
    Bone Sampling Mode=::NewEnumerator4
    MaskSpawnGroup=false
    Position Sampling=::NewEnumerator2
    Skeletal Mesh=$System.Skeletal Mesh
    Use Sampled Mesh Attribute as Ribbon ID=false
   )
   /SetVariables(Particles.CameraDirectionFloat:NiagaraFloat=&FN_Fit(
     Clamp:NiagaraBool=false
     Value:NiagaraFloat=&FN_Fit(
      Clamp:NiagaraBool=false
      Value:NiagaraFloat=&DotProduct(
       Vector A=&ReturnCameraVector(Camera Vectors=::NewEnumerator1, Coordinate Space=::Local)
       Vector B=&ReturnMeshOrientationAxis(Mesh Orientation Axis=::YAxis, Mesh Orientation Quat=$Particles.MeshOrientation)
      )
      SourceMin:NiagaraFloat=-1
      SourceMax:NiagaraFloat=1
      DestMin:NiagaraFloat=0
      DestMax:NiagaraFloat=1
     )
     SourceMin:NiagaraFloat=-1
     SourceMax:NiagaraFloat=1
     DestMin:NiagaraFloat=0
     DestMax:NiagaraFloat=1
    )) {isSetParameters=true}
   /SetVariables() {enabled=false, isSetParameters=true}
   &InterpolateOverTime(Rate Of Change=3, Target Float Value:NiagaraFloat=$Particles.CameraDirectionFloat)
   &SolveForcesAndVelocity(Clamp Velocity=false, Limit Acceleration=false)
   &self:Set_position_on_Ribbon(
    Particle Attribute Reader=DataInterface() {propertyValues {EmitterBinding {bindingMode=Other, emitterName=Stem_Ribbons}}}
    ParticlePlacement:NiagaraInt32=&MakeIntFromFloat(Float=&MergeHelper_FloatSwitch(
      0=&FN_Fit(
       Clamp:NiagaraBool=true
       Value:NiagaraFloat=&FloatFromCurve(
        FloatCurve=DataInterface() {propertyValues {Curve {keys [<at 0 0.99>,<at 0.2923 0.8845>,<at 0.7418 0.7361>,<at 1 0.5>]}}}
        CurveIndex=$Particles.PartitionParticles.Normalized Index In Partition
       )
       SourceMin:NiagaraFloat=0
       SourceMax:NiagaraFloat=1
       DestMin:NiagaraFloat=&MakeFloatFromInt(Int=&Multiply_Int(A=$Particles.PartitionParticles.Partition, B=$System.ParticlesPerRibbon))
       DestMax:NiagaraFloat=&MakeFloatFromInt(Int=/Niagara/DynamicInputs/Subtract/Subtract_Integer(
         A:NiagaraInt32=&Multiply_Int(A=&Add_Integer(A=$Particles.PartitionParticles.Partition, B=1), B=$System.ParticlesPerRibbon)
         B:NiagaraInt32=1
        ))
      )
      1=&FN_Fit(
       Clamp:NiagaraBool=true
       Value:NiagaraFloat=&FloatFromCurve(
        FloatCurve=DataInterface() {propertyValues {Curve {keys [<at 0 0.99>,<at 0.2389 0.9228>,<at 0.6544 0.7776>,<at 1.0037 0.4957>]}}}
        CurveIndex=$Particles.PartitionParticles.Normalized Index In Partition
       )
       SourceMin:NiagaraFloat=0
       SourceMax:NiagaraFloat=1
       DestMin:NiagaraFloat=&MakeFloatFromInt(Int=&Multiply_Int(A=$Particles.PartitionParticles.Partition, B=$System.ParticlesPerRibbon))
       DestMax:NiagaraFloat=&MakeFloatFromInt(Int=&Subtract_Integer(
         A:NiagaraInt32=&Multiply_Int(A=&Add_Integer(A=$Particles.PartitionParticles.Partition, B=1), B=$System.ParticlesPerRibbon)
         B:NiagaraInt32=1
        ))
      )
      2=&FN_Fit(
       Clamp:NiagaraBool=true
       Value:NiagaraFloat=&FloatFromCurve(
        FloatCurve=DataInterface() {propertyValues {Curve {keys [<at 0 0.99>,<at 0.285 0.896>,<at 0.7343 0.7467>,<at 1 0.6074>]}}}
        CurveIndex=$Particles.PartitionParticles.Normalized Index In Partition
       )
       SourceMin:NiagaraFloat=0
       SourceMax:NiagaraFloat=1
       DestMin:NiagaraFloat=&MakeFloatFromInt(Int=&Multiply_Int(A=$Particles.PartitionParticles.Partition, B=$System.ParticlesPerRibbon))
       DestMax:NiagaraFloat=&MakeFloatFromInt(Int=&Subtract_Integer(
         A:NiagaraInt32=&Multiply_Int(A=&Add_Integer(A=$Particles.PartitionParticles.Partition, B=1), B=$System.ParticlesPerRibbon)
         B:NiagaraInt32=1
        ))
      )
      3=&FN_Fit(
       Clamp:NiagaraBool=true
       Value:NiagaraFloat=&FloatFromCurve(
        FloatCurve=DataInterface() {propertyValues {Curve {keys [<at 0 0.99>,<at 0.2713 0.8016>,<at 0.7355 0.6279>,<at 1 0.4998>]}}}
        CurveIndex=$Particles.PartitionParticles.Normalized Index In Partition
       )
       SourceMin:NiagaraFloat=0
       SourceMax:NiagaraFloat=1
       DestMin:NiagaraFloat=&MakeFloatFromInt(Int=&Multiply_Int(A=$Particles.PartitionParticles.Partition, B=$System.ParticlesPerRibbon))
       DestMax:NiagaraFloat=&MakeFloatFromInt(Int=&Subtract_Integer(
         A:NiagaraInt32=&Multiply_Int(A=&Add_Integer(A=$Particles.PartitionParticles.Partition, B=1), B=$System.ParticlesPerRibbon)
         B:NiagaraInt32=1
        ))
      )
      4=&FN_Fit(
       Clamp:NiagaraBool=true
       Value:NiagaraFloat=&FloatFromCurve(
        FloatCurve=DataInterface() {propertyValues {Curve {keys [<at 0 0.99>,<at 0.304 0.8675>,<at 0.7177 0.7141>,<at 0.9921 0.5676>]}}}
        CurveIndex=$Particles.PartitionParticles.Normalized Index In Partition
       )
       SourceMin:NiagaraFloat=0
       SourceMax:NiagaraFloat=1
       DestMin:NiagaraFloat=&MakeFloatFromInt(Int=&Multiply_Int(A=$Particles.PartitionParticles.Partition, B=$System.ParticlesPerRibbon))
       DestMax:NiagaraFloat=&MakeFloatFromInt(Int=&Subtract_Integer(
         A:NiagaraInt32=&Multiply_Int(A=&Add_Integer(A=$Particles.PartitionParticles.Partition, B=1), B=$System.ParticlesPerRibbon)
         B:NiagaraInt32=1
        ))
      )
      5=&FN_Fit(
       Clamp:NiagaraBool=true
       Value:NiagaraFloat=&FloatFromCurve(
        FloatCurve=DataInterface() {propertyValues {Curve {keys [<at 0 0.99>,<at 0.3 0.9168>,<at 0.6081 0.6696>,<at 1 0.5319>]}}}
        CurveIndex=$Particles.PartitionParticles.Normalized Index In Partition
       )
       SourceMin:NiagaraFloat=0
       SourceMax:NiagaraFloat=1
       DestMin:NiagaraFloat=&MakeFloatFromInt(Int=&Multiply_Int(A=$Particles.PartitionParticles.Partition, B=$System.ParticlesPerRibbon))
       DestMax:NiagaraFloat=&MakeFloatFromInt(Int=&Subtract_Integer(
         A:NiagaraInt32=&Multiply_Int(A=&Add_Integer(A=$Particles.PartitionParticles.Partition, B=1), B=$System.ParticlesPerRibbon)
         B:NiagaraInt32=1
        ))
      )
      Index=$Particles.PartitionParticles.Partition
      Num=::NewEnumerator4
     ))
   ) {module=Set_position_on_Ribbon}
   /SetVariables(Particles.LeafColor:LinearColor=&MergeHelper_ColorSwitch(
     0=&ColorFromCurve(
      ColorCurve=DataInterface() {propertyValues {
        AlphaCurve {keys [<at 0 1 RCIM_Cubic>]}
        BlueCurve {keys [<at 0.3064 0 RCIM_Cubic>,<at 1 0 RCIM_Cubic>]}
        GreenCurve {keys [<at 0.3064 0 RCIM_Cubic>,<at 1 0.0313 RCIM_Cubic>]}
        RedCurve {keys [<at 0.3064 0.7734 RCIM_Cubic>,<at 1 0.75 RCIM_Cubic>]}
       }}
      CurveIndex=$Particles.PartitionParticles.Normalized Index In Partition
     )
     1=&ColorFromCurve(
      ColorCurve=DataInterface() {propertyValues {
        AlphaCurve {keys [<at 0 1 RCIM_Cubic>]}
        BlueCurve {keys [
          <at 0 0.4634 RCIM_Cubic>
          {arriveTangent=-0.3144, interpMode=RCIM_Cubic, leaveTangent=-0.3144, time=0.6351, value=0.1563}
          <at 0.9976 0.1498>
         ]}
        GreenCurve {keys [<at 0 0 RCIM_Cubic>,<at 0.6351 0 RCIM_Cubic>,<at 0.9976 0>]}
        RedCurve {keys [<at 0 1 RCIM_Cubic>,<at 0.6351 0.9349 RCIM_Cubic>,<at 0.9976 1>]}
       }}
      CurveIndex=$Particles.PartitionParticles.Normalized Index In Partition
     )
     2=&ColorFromCurve(
      ColorCurve=DataInterface() {propertyValues {
        AlphaCurve {keys [<at 0 1 RCIM_Cubic>]}
        BlueCurve {keys [<at 0.3064 0 RCIM_Cubic>,<at 1 0 RCIM_Cubic>]}
        GreenCurve {keys [<at 0.3064 0 RCIM_Cubic>,<at 1 0.0313 RCIM_Cubic>]}
        RedCurve {keys [<at 0.3064 0.7734 RCIM_Cubic>,<at 1 0.75 RCIM_Cubic>]}
       }}
      CurveIndex=$Particles.PartitionParticles.Normalized Index In Partition
     )
     3=&ColorFromCurve(
      ColorCurve=DataInterface() {propertyValues {
        AlphaCurve {keys [<at 0 1 RCIM_Cubic>]}
        BlueCurve {keys [<at 0.3064 0 RCIM_Cubic>,<at 1 0 RCIM_Cubic>]}
        GreenCurve {keys [<at 0.3064 0 RCIM_Cubic>,<at 1 0.0313 RCIM_Cubic>]}
        RedCurve {keys [<at 0.3064 0.7734 RCIM_Cubic>,<at 1 0.75 RCIM_Cubic>]}
       }}
      CurveIndex=$Particles.PartitionParticles.Normalized Index In Partition
     )
     4=&ColorFromCurve(
      ColorCurve=DataInterface() {propertyValues {
        AlphaCurve {keys [<at 0 1 RCIM_Cubic>]}
        BlueCurve {keys [
          <at 0 0.4634 RCIM_Cubic>
          {arriveTangent=-0.3144, interpMode=RCIM_Cubic, leaveTangent=-0.3144, time=0.6351, value=0.1563}
          <at 0.9976 0.1498>
         ]}
        GreenCurve {keys [<at 0 0 RCIM_Cubic>,<at 0.6351 0 RCIM_Cubic>,<at 0.9976 0>]}
        RedCurve {keys [<at 0 1 RCIM_Cubic>,<at 0.6351 0.9349 RCIM_Cubic>,<at 0.9976 1>]}
       }}
      CurveIndex=$Particles.PartitionParticles.Normalized Index In Partition
     )
     5=&ColorFromCurve(
      ColorCurve=DataInterface() {propertyValues {
        AlphaCurve {keys [<at 0 1 RCIM_Cubic>]}
        BlueCurve {keys [<at 0.3064 0 RCIM_Cubic>,<at 1 0 RCIM_Cubic>]}
        GreenCurve {keys [<at 0.3064 0 RCIM_Cubic>,<at 1 0.0313 RCIM_Cubic>]}
        RedCurve {keys [<at 0.3064 0.7734 RCIM_Cubic>,<at 1 0.75 RCIM_Cubic>]}
       }}
      CurveIndex=$Particles.PartitionParticles.Normalized Index In Partition
     )
     Index=$Particles.PartitionParticles.Partition
     Num=::NewEnumerator4
    )) {isSetParameters=true}
   /SetVariables(Particles.rotation:NiagaraFloat=&Add_Float(
     A=&Add_Float(
      A=0
      B=&Lerp_Float(A=$Particles.InitialRotationB, B=$Particles.InitialRotation A, Alpha=$Particles.InterpolateOverTime.InterpolateOverTime_Float.Moving Average)
     )
     B=&Multiply_Float(
      A=&Waveform(
       Global Phase=&Add_Float(A=&RandomRangeFloat(Minimum=0, Maximum=1), B=$Engine.Time)
       "Waveform [1]"=::NewEnumerator12
       "[1] Frequency"=0.75
       "[1] Amplitude Min/Max"=<-25 25>
      )
      B=$Particles.InterpolateOverTime.InterpolateOverTime_Float.Moving Average
     )
    )) {isSetParameters=true}
   &DynamicMaterialParameters(Index 0 Linear Color:LinearColor=$Particles.RibbonColorSample, Write Parameter Index 0=::NewEnumerator5)
  ]
  rendererClasses [&NiagaraSpriteRendererProperties]
  renderers [{rendererClass=&NiagaraSpriteRendererProperties}]
 }
 Spawn_Icons_attached {
  bEnabled=false
  emitterUpdateScript [
   &EmitterState()
   &SpawnRate(SpawnRate=6.5, Use Spawn Probability=false)
   &SpawnBurst_Instantaneous(Spawn Count=7, Use Spawn Probability=false, UseLoopCountLimit=true)
   /SetVariables(Emitter.Icon Scale Mod:NiagaraFloat=&MakeCustomFloatFromBool(Bool=$User.IsLargeBody, True Float=1.1, False Float=0.95)) {isSetParameters=true, module=SetVariables_#6}
  ]
  particleSpawnScript [
   &PartitionParticles(Number Of Partitions=4)
   &PartitionParticles(Number Of Partitions=14) {module=PartitionParticles001}
   &InitializeParticle(
    Color Mode=::NewEnumerator0
    Lifetime=&RandomRangeFloat(Minimum=1.75, Maximum=2.25)
    Sprite Size Mode=::NewEnumerator3
    Uniform Sprite Size=&Multiply_Float(A=&RandomRangeFloat(Minimum=9, Maximum=10), B=$Emitter.Icon Scale Mod)
    UsePositionOffset=false
    Write Material Random=true
   )
   &CameraOffset(Camera Offset Amount=-10)
   /SetVariables(Particles.FramePicker:NiagaraInt32=&SelectIntFromArray(
     Direct Array Index=$Particles.PartitionParticles001.Partition
     Int Selection Array=DataInterface() {propertyValues {IntData [0,4,6,8,10,12,14,1,5,7,9,11,13,15]}}
    )) {isSetParameters=true, module=SetVariables_#7}
   /SetVariables(
    Particles.PivotOffsetPicker:Vector2f=&SelectVector2DFromArray(
     Direct Array Index=$Particles.FramePicker
     Vector 2D Selection Array=DataInterface() {propertyValues {FloatData [
        <0.5 0.95>
        <0.5 0.95>
        <0.4 0.8>
        <0.6 0.8>
        <0.43 0.98>
        <0.57 0.98>
        <0.5 0.95>
        <0.5 0.95>
        <0.5 0.9>
        <0.5 0.9>
        <0.5 0.9>
        <0.5 0.9>
        <0.5 0.93>
        <0.5 0.93>
        <0.5 0.93>
        <0.5 0.93>
        <0.5 0.9>
       ]}}
    )
    Particles.SpriteSize=&Multiply_Vector2D_ByFloat(
     Float=&SelectFloatFromArray(
      Direct Array Index=$Particles.FramePicker
      Float Selection Array=DataInterface() {propertyValues {FloatData [1.2,1.2,0.5,0.5,1.3,1.3,1,1,0.75,0.75,1,1,0.75,0.75,0.75,0.75,1]}}
     )
     Vector2D=$Particles.SpriteSize
    )
    Particles.StartAngle:NiagaraFloat=&RandomRangeFloat(Minimum=-10, Maximum=10)
    Particles.UVScalePicker:Vector2f=&SelectVector2DFromArray(
     Direct Array Index=$Particles.FramePicker
     Vector 2D Selection Array=DataInterface() {propertyValues {FloatData [<1 1>, <-1 1>, <1 1>, <-1 1>, <1 1>, <-1 1>, <1 1>, <-1 1>, <1 1>, <-1 1>, <1 1>, <-1 1>, <1 1>, <-1 1>, <1 1>, <-1 1>, <1 1>]}}
    )
   ) {isSetParameters=true, module=SetVariables_#8}
   &SubUVAnimation(
    Direct Frame Index:NiagaraFloat=&SelectFloatFromArray(
     Direct Array Index=$Particles.FramePicker
     Float Selection Array=DataInterface() {propertyValues {FloatData [7,7,8,8,9,9,10,10,11,11,12,12,13,13,14,14,15]}}
    )
    "Sprite Renderer "=DataInterface() {propertyValues {SpriteRenderer=&self:Spawn_Icons_attached_0.NiagaraSpriteRendererProperties_1}}
    SubUV Animation Mode:ENiagara_SubUVLookupModeV2=::NewEnumerator5
   )
  ]
  particleUpdateScript [
   &ParticleState()
   &ScaleSpriteSize(
    Scale Factor=&MakeVector2D(
     X=&FloatFromCurve(FloatCurve=DataInterface() {propertyValues {Curve {
         defaultValue=0
         keys [
          {arriveTangent=5.7026, interpMode=RCIM_Cubic, leaveTangent=5.7026, tangentMode=RCTM_User, time=0, value=0}
          {interpMode=RCIM_Cubic, tangentMode=RCTM_User, time=0.4, value=1}
         ]
        }}})
     Y=&FloatFromCurve(FloatCurve=DataInterface() {propertyValues {Curve {
         defaultValue=0
         keys [
          {interpMode=RCIM_Cubic, tangentMode=RCTM_User, time=0, value=0.027}
          {arriveTangent=0.5498, interpMode=RCIM_Cubic, leaveTangent=0.5498, tangentMode=RCTM_User, time=0.5, value=1}
          <at 0.9981 1.1 RCIM_Cubic>
         ]
        }}})
    )
    Scale Sprite Size Mode=::NewEnumerator2
   )
   &SkeletalMeshLocation(
    Bone / Socket Index=&SelectIntFromArray(Direct Array Index=$Particles.PartitionParticles.Partition, Int Selection Array=DataInterface() {propertyValues {IntData [1,2,7,7]}})
    Bone Sampling Mode=::NewEnumerator4
    MaskSpawnGroup=false
    Sampled Position Offset=&Multiply_Vector(
     A=&MergeHelper_VectorSwitch(
      0=&MakeVector(
       X=&FloatFromCurve(
        FloatCurve=DataInterface() {propertyValues {Curve {defaultValue=1, keys [{tangentMode=RCTM_User, time=0, value=0.1}, {tangentMode=RCTM_User, time=1, value=1}]}}}
        CurveIndex=&RandomRangeFloat(Minimum=0, Maximum=1)
        Scale Curve=&MakeCustomFloatFromBool(Bool=$User.IsLargeBody, True Float=14, False Float=13)
       )
       Y=&FloatFromCurve(
        FloatCurve=DataInterface() {propertyValues {Curve {keys [<at 0 1>,<at 1 0.1>]}}}
        CurveIndex=&RandomRangeFloat(Minimum=0, Maximum=1)
        Scale Curve=-2
       )
       Z=&FloatFromCurve(
        FloatCurve=DataInterface() {propertyValues {Curve {keys [<at 0 1>,<at 1 0.1>]}}}
        CurveIndex=&RandomRangeFloat(Minimum=0, Maximum=1)
        Scale Curve=2
       )
      )
      1=&MakeVector(
       X=&FloatFromCurve(
        FloatCurve=DataInterface() {propertyValues {Curve {defaultValue=1, keys [{tangentMode=RCTM_User, time=0, value=0.1}, {tangentMode=RCTM_User, time=1, value=1}]}}}
        CurveIndex=&RandomRangeFloat(Minimum=0, Maximum=1)
        Scale Curve=&MakeCustomFloatFromBool(Bool=$User.IsLargeBody, True Float=-14, False Float=-13)
       )
       Y=&FloatFromCurve(
        FloatCurve=DataInterface() {propertyValues {Curve {keys [<at 0 1>,<at 1 0.1>]}}}
        CurveIndex=&RandomRangeFloat(Minimum=0, Maximum=1)
        Scale Curve=2
       )
       Z=&FloatFromCurve(
        FloatCurve=DataInterface() {propertyValues {Curve {keys [<at 0 1>,<at 1 0.1>]}}}
        CurveIndex=&RandomRangeFloat(Minimum=0, Maximum=1)
        Scale Curve=-2
       )
      )
      2=&MakeVector(
       X=&FloatFromCurve(
        FloatCurve=DataInterface() {propertyValues {Curve {defaultValue=1, keys [{tangentMode=RCTM_User, time=0, value=1}, {tangentMode=RCTM_User, time=1, value=1}]}}}
        CurveIndex=&RandomRangeFloat(Minimum=0, Maximum=1)
        Scale Curve=3
       )
       Y=&FloatFromCurve(
        FloatCurve=DataInterface() {propertyValues {Curve {keys [<at 0 0>,<at 1 0>]}}}
        CurveIndex=&RandomRangeFloat(Minimum=0, Maximum=1)
        Scale Curve=3
       )
       Z=&FloatFromCurve(
        FloatCurve=DataInterface() {propertyValues {Curve {keys [<at 0 0.45>,<at 1 1>]}}}
        CurveIndex=&RandomRangeFloat(Minimum=0, Maximum=1)
        Scale Curve=&MakeCustomFloatFromBool(Bool=$User.IsLargeBody, True Float=20, False Float=14)
       )
      )
      3=&MakeVector(
       X=&FloatFromCurve(
        FloatCurve=DataInterface() {propertyValues {Curve {defaultValue=1, keys [{tangentMode=RCTM_User, time=0, value=1}, {tangentMode=RCTM_User, time=1, value=1}]}}}
        CurveIndex=&RandomRangeFloat(Minimum=0, Maximum=1)
        Scale Curve=3
       )
       Y=&FloatFromCurve(
        FloatCurve=DataInterface() {propertyValues {Curve {keys [<at 0 0>,<at 1 0>]}}}
        CurveIndex=&RandomRangeFloat(Minimum=0, Maximum=1)
        Scale Curve=3
       )
       Z=&FloatFromCurve(
        FloatCurve=DataInterface() {propertyValues {Curve {keys [<at 0 -0.45>,<at 1 -1>]}}}
        CurveIndex=&RandomRangeFloat(Minimum=0, Maximum=1)
        Scale Curve=&MakeCustomFloatFromBool(Bool=$User.IsLargeBody, True Float=20, False Float=14)
       )
      )
      Index=$Particles.PartitionParticles.Partition
      Num=::NewEnumerator2
     )
     B=<1 1 1>
    )
    Skeletal Mesh=$System.Skeletal Mesh
    Use Sampled Mesh Attribute as Ribbon ID=false
   ) {module=SkeletalMeshLocation001}
   /SetVariables(Particles.PositionOffset:NiagaraPosition=&AddVectorToPosition(
     Position=$Particles.Position
     Vector=&Add_Vector(
      A=&Multiply_Vector(
       A=&Add_Vector(
        A=&MergeHelper_VectorSwitch(
         0=<0 0 0>
         1=<0 0 0>
         2=&MakeVector(X=0, Y=0, Z=&MakeCustomFloatFromBool(Bool=$User.IsFrontEnd, True Float=3, False Float=1.5))
         3=&MakeVector(X=0, Y=0, Z=&MakeCustomFloatFromBool(Bool=$User.IsFrontEnd, True Float=3, False Float=1.5))
         Index=$Particles.PartitionParticles.Partition
         Num=::NewEnumerator2
        )
        B=&MakeCustomVectorFromBool(
         Bool=$User.IsFrontEnd
         True Vector=&MakeCustomVectorFromBool(Bool=$User.IsLargeBody, True Vector=<0 0 2>, False Vector=<0 0 1>)
         False Vector=<0 0 2>
        )
       )
       B=<1 1 1>
      )
      B=&MakeVector(
       X=0
       Y=0
       Z=&FloatFromCurve(
        FloatCurve=DataInterface() {propertyValues {Curve {
           defaultValue=0
           keys [{interpMode=RCIM_Cubic, tangentMode=RCTM_User, time=0, value=0}, {interpMode=RCIM_Cubic, tangentMode=RCTM_User, time=1, value=1}]
          }}}
        Scale Curve=&MakeCustomFloatFromBool(
         Bool=$User.IsLargeBody
         True Float=&RandomRangeFloat(Minimum=1, Maximum=10)
         False Float=&RandomRangeFloat(Minimum=0, Maximum=8)
        )
       )
      )
     )
    )) {isSetParameters=true, module=SetVariables_#9}
   /SetVariables(Particles.Debugrotation:NiagaraFloat=&Waveform(
     Global Phase=&Add_Float(A=&RandomRangeFloat(Minimum=0, Maximum=1), B=$Engine.Time)
     "Waveform [1]"=::NewEnumerator12
     "[1] Phase Scale"=2
     "[1] Frequency"=0.5
     "[1] Amplitude Min/Max"=<-20 20>
    )) {isSetParameters=true, module=SetVariables_#10}
   /SetVariables(Particles.SpriteRotation:NiagaraFloat=&Add_Float(A=$Particles.Debugrotation, B=$Particles.StartAngle)) {isSetParameters=true, module=SetVariables_#11}
   &ScaleColor(
    Scale Alpha=&Multiply_Float(
     A=&FloatFromCurve(
      FloatCurve=DataInterface() {propertyValues {Curve {
         defaultValue=1
         keys [
          {arriveTangent=13.9519, interpMode=RCIM_Cubic, leaveTangent=13.9519, tangentMode=RCTM_User, time=0, value=0}
          {arriveTangentWeight=0.0469, interpMode=RCIM_Cubic, leaveTangentWeight=0.1864, tangentMode=RCTM_User, time=0.1, value=0.9847}
          {arriveTangentWeight=0.2835, interpMode=RCIM_Cubic, leaveTangentWeight=0.1521, tangentMode=RCTM_User, time=0.55, value=1}
          {arriveTangentWeight=0.676, interpMode=RCIM_Cubic, tangentMode=RCTM_User, time=1, value=0}
         ]
        }}}
      Scale Curve=&MakeCustomFloatFromBool(Bool=$User.IsFrontEnd, True Float=0.8, False Float=1)
     )
     B=&Multiply_Float(
      A=&Waveform(
       Global Phase=&Add_Float(A=$Emitter.Age, B=&RandomRangeFloat(Minimum=0, Maximum=1))
       "Waveform [1]"=::NewEnumerator9
       "[1] Frequency"=1
       "[1] Amplitude Min/Max"=<0.7 1>
      )
      B=&MakeCustomFloatFromBool(
       Bool=&SetBoolByFloatComparison(
        A=&OneMinusFloat(Float=&DotProduct(Vector A=&ReturnCameraVector(Camera Vectors=::NewEnumerator1, Coordinate Space=::Local), Vector B=<0 0 1>))
        B=0.7
       )
       True Float=&FloatFromCurve(
        FloatCurve=DataInterface() {propertyValues {Curve {keys [<at 0.7 1>,<at 0.95 0>]}}}
        CurveIndex=&Abs_Float(Float:NiagaraFloat=&DotProduct(Vector A=&ReturnCameraVector(Camera Vectors=::NewEnumerator1, Coordinate Space=::Local), Vector B=<0 0 1>))
       )
       False Float=1
      )
     )
    )
    Scale RGB=&Add_Vector(A=$System.Color RGB, B=<0.5 0.5 0.5>)
    ScaleA=true
    ScaleRGB=true
   )
  ]
  rendererClasses [&NiagaraSpriteRendererProperties, &NiagaraMeshRendererProperties]
  renderers [
   {rendererClass=&NiagaraSpriteRendererProperties}
   {rendererClass=&NiagaraMeshRendererProperties, rendererIndex=1}
   {rendererClass=&NiagaraSpriteRendererProperties, rendererIndex=2}
  ]
 }
 Spawn_Icons_attached001 {
  bEnabled=false
  emitterUpdateScript [
   &EmitterState()
   &SpawnRate(SpawnRate=6.5, Use Spawn Probability=false)
   &SpawnBurst_Instantaneous(Spawn Count=5, Use Spawn Probability=false, UseLoopCountLimit=true)
   /SetVariables(Emitter.Icon Scale Mod:NiagaraFloat=&MakeCustomFloatFromBool(Bool=$User.IsLargeBody, True Float=1.05, False Float=0.9)) {isSetParameters=true, module=SetVariables_#6}
  ]
  particleSpawnScript [
   &PartitionParticles(Number Of Partitions=4)
   &PartitionParticles(Number Of Partitions=10) {module=PartitionParticles001}
   &InitializeParticle(
    Color Mode=::NewEnumerator0
    Lifetime=&RandomRangeFloat(Minimum=1.75, Maximum=2.25)
    Sprite Size Mode=::NewEnumerator3
    Uniform Sprite Size=&Multiply_Float(A=&RandomRangeFloat(Minimum=9, Maximum=10), B=$Emitter.Icon Scale Mod)
    UsePositionOffset=false
    Write Material Random=true
   )
   &CameraOffset(Camera Offset Amount=-10)
   /SetVariables(Particles.FramePicker:NiagaraInt32=&SelectIntFromArray(
     Direct Array Index=$Particles.PartitionParticles001.Partition
     Int Selection Array=DataInterface() {propertyValues {IntData [0,15,4,10,14,11,1,12,5,13]}}
    )) {isSetParameters=true, module=SetVariables_#7}
   /SetVariables(
    Particles.PivotOffsetPicker:Vector2f=&SelectVector2DFromArray(
     Direct Array Index=$Particles.FramePicker
     Vector 2D Selection Array=DataInterface() {propertyValues {FloatData [
        <0.5 0.95>
        <0.5 0.95>
        <0.4 0.8>
        <0.6 0.8>
        <0.43 0.98>
        <0.57 0.98>
        <0.5 0.95>
        <0.5 0.95>
        <0.5 0.9>
        <0.5 0.9>
        <0.5 0.9>
        <0.5 0.9>
        <0.5 0.93>
        <0.5 0.93>
        <0.5 0.93>
        <0.5 0.93>
        <0.5 0.9>
       ]}}
    )
    Particles.SpriteSize=&Multiply_Vector2D_ByFloat(
     Float=&SelectFloatFromArray(
      Direct Array Index=$Particles.FramePicker
      Float Selection Array=DataInterface() {propertyValues {FloatData [1.2,1.2,0.5,0.5,1.3,1.3,1,1,0.75,0.75,1,1,0.75,0.75,0.75,0.75,1]}}
     )
     Vector2D=$Particles.SpriteSize
    )
    Particles.StartAngle:NiagaraFloat=&RandomRangeFloat(Minimum=-10, Maximum=10)
    Particles.UVScalePicker:Vector2f=&SelectVector2DFromArray(
     Direct Array Index=$Particles.FramePicker
     Vector 2D Selection Array=DataInterface() {propertyValues {FloatData [<1 1>, <-1 1>, <1 1>, <-1 1>, <1 1>, <-1 1>, <1 1>, <-1 1>, <1 1>, <-1 1>, <1 1>, <-1 1>, <1 1>, <-1 1>, <1 1>, <-1 1>, <1 1>]}}
    )
   ) {isSetParameters=true, module=SetVariables_#8}
   &SubUVAnimation(
    Direct Frame Index:NiagaraFloat=&SelectFloatFromArray(
     Direct Array Index=$Particles.FramePicker
     Float Selection Array=DataInterface() {propertyValues {FloatData [7,7,8,8,9,9,10,10,11,11,12,12,13,13,14,14,15]}}
    )
    "Sprite Renderer "=DataInterface() {propertyValues {SpriteRenderer=&self:Spawn_Icons_attached001_0.NiagaraSpriteRendererProperties_1}}
    SubUV Animation Mode:ENiagara_SubUVLookupModeV2=::NewEnumerator5
   )
  ]
  particleUpdateScript [
   &ParticleState()
   &ScaleSpriteSize(
    Scale Factor=&MakeVector2D(
     X=&FloatFromCurve(FloatCurve=DataInterface() {propertyValues {Curve {
         defaultValue=0
         keys [
          {arriveTangent=5.7026, interpMode=RCIM_Cubic, leaveTangent=5.7026, tangentMode=RCTM_User, time=0, value=0}
          {interpMode=RCIM_Cubic, tangentMode=RCTM_User, time=0.4, value=1}
         ]
        }}})
     Y=&FloatFromCurve(FloatCurve=DataInterface() {propertyValues {Curve {
         defaultValue=0
         keys [
          {interpMode=RCIM_Cubic, tangentMode=RCTM_User, time=0, value=0.027}
          {arriveTangent=0.5498, interpMode=RCIM_Cubic, leaveTangent=0.5498, tangentMode=RCTM_User, time=0.5, value=1}
          <at 0.9981 1.1 RCIM_Cubic>
         ]
        }}})
    )
    Scale Sprite Size Mode=::NewEnumerator2
   )
   &SkeletalMeshLocation(
    Bone / Socket Index=&SelectIntFromArray(Direct Array Index=$Particles.PartitionParticles.Partition, Int Selection Array=DataInterface() {propertyValues {IntData [1,2,7,7]}})
    Bone Sampling Mode=::NewEnumerator4
    MaskSpawnGroup=false
    Sampled Position Offset=&Multiply_Vector(
     A=&MergeHelper_VectorSwitch(
      0=&MakeVector(
       X=&FloatFromCurve(
        FloatCurve=DataInterface() {propertyValues {Curve {defaultValue=1, keys [{tangentMode=RCTM_User, time=0, value=0.1}, {tangentMode=RCTM_User, time=1, value=1}]}}}
        CurveIndex=&RandomRangeFloat(Minimum=0, Maximum=1)
        Scale Curve=&MakeCustomFloatFromBool(Bool=$User.IsLargeBody, True Float=14, False Float=13)
       )
       Y=&FloatFromCurve(
        FloatCurve=DataInterface() {propertyValues {Curve {keys [<at 0 1>,<at 1 0.1>]}}}
        CurveIndex=&RandomRangeFloat(Minimum=0, Maximum=1)
        Scale Curve=-2
       )
       Z=&FloatFromCurve(
        FloatCurve=DataInterface() {propertyValues {Curve {keys [<at 0 1>,<at 1 0.1>]}}}
        CurveIndex=&RandomRangeFloat(Minimum=0, Maximum=1)
        Scale Curve=2
       )
      )
      1=&MakeVector(
       X=&FloatFromCurve(
        FloatCurve=DataInterface() {propertyValues {Curve {defaultValue=1, keys [{tangentMode=RCTM_User, time=0, value=0.1}, {tangentMode=RCTM_User, time=1, value=1}]}}}
        CurveIndex=&RandomRangeFloat(Minimum=0, Maximum=1)
        Scale Curve=&MakeCustomFloatFromBool(Bool=$User.IsLargeBody, True Float=-14, False Float=-13)
       )
       Y=&FloatFromCurve(
        FloatCurve=DataInterface() {propertyValues {Curve {keys [<at 0 1>,<at 1 0.1>]}}}
        CurveIndex=&RandomRangeFloat(Minimum=0, Maximum=1)
        Scale Curve=2
       )
       Z=&FloatFromCurve(
        FloatCurve=DataInterface() {propertyValues {Curve {keys [<at 0 1>,<at 1 0.1>]}}}
        CurveIndex=&RandomRangeFloat(Minimum=0, Maximum=1)
        Scale Curve=-2
       )
      )
      2=&MakeVector(
       X=&FloatFromCurve(
        FloatCurve=DataInterface() {propertyValues {Curve {defaultValue=1, keys [{tangentMode=RCTM_User, time=0, value=1}, {tangentMode=RCTM_User, time=1, value=1}]}}}
        CurveIndex=&RandomRangeFloat(Minimum=0, Maximum=1)
        Scale Curve=3
       )
       Y=&FloatFromCurve(
        FloatCurve=DataInterface() {propertyValues {Curve {keys [<at 0 0>,<at 1 0>]}}}
        CurveIndex=&RandomRangeFloat(Minimum=0, Maximum=1)
        Scale Curve=3
       )
       Z=&FloatFromCurve(
        FloatCurve=DataInterface() {propertyValues {Curve {keys [<at 0 0.45>,<at 1 1>]}}}
        CurveIndex=&RandomRangeFloat(Minimum=0, Maximum=1)
        Scale Curve=&MakeCustomFloatFromBool(Bool=$User.IsLargeBody, True Float=20, False Float=14)
       )
      )
      3=&MakeVector(
       X=&FloatFromCurve(
        FloatCurve=DataInterface() {propertyValues {Curve {defaultValue=1, keys [{tangentMode=RCTM_User, time=0, value=1}, {tangentMode=RCTM_User, time=1, value=1}]}}}
        CurveIndex=&RandomRangeFloat(Minimum=0, Maximum=1)
        Scale Curve=3
       )
       Y=&FloatFromCurve(
        FloatCurve=DataInterface() {propertyValues {Curve {keys [<at 0 0>,<at 1 0>]}}}
        CurveIndex=&RandomRangeFloat(Minimum=0, Maximum=1)
        Scale Curve=3
       )
       Z=&FloatFromCurve(
        FloatCurve=DataInterface() {propertyValues {Curve {keys [<at 0 -0.45>,<at 1 -1>]}}}
        CurveIndex=&RandomRangeFloat(Minimum=0, Maximum=1)
        Scale Curve=&MakeCustomFloatFromBool(Bool=$User.IsLargeBody, True Float=20, False Float=14)
       )
      )
      Index=$Particles.PartitionParticles.Partition
      Num=::NewEnumerator2
     )
     B=<1 1 1>
    )
    Skeletal Mesh=$System.Skeletal Mesh
    Use Sampled Mesh Attribute as Ribbon ID=false
   ) {module=SkeletalMeshLocation001}
   /SetVariables(Particles.PositionOffset:NiagaraPosition=&AddVectorToPosition(
     Position=$Particles.Position
     Vector=&Add_Vector(
      A=&Multiply_Vector(
       A=&Add_Vector(
        A=&MergeHelper_VectorSwitch(
         0=<0 0 0>
         1=<0 0 0>
         2=&MakeVector(X=0, Y=0, Z=&MakeCustomFloatFromBool(Bool=$User.IsFrontEnd, True Float=3, False Float=1.5))
         3=&MakeVector(X=0, Y=0, Z=&MakeCustomFloatFromBool(Bool=$User.IsFrontEnd, True Float=3, False Float=1.5))
         Index=$Particles.PartitionParticles.Partition
         Num=::NewEnumerator2
        )
        B=&MakeCustomVectorFromBool(
         Bool=$User.IsFrontEnd
         True Vector=&MakeCustomVectorFromBool(Bool=$User.IsLargeBody, True Vector=<0 0 2>, False Vector=<0 0 1>)
         False Vector=<0 0 2>
        )
       )
       B=<1 1 1>
      )
      B=&MakeVector(
       X=0
       Y=0
       Z=&FloatFromCurve(
        FloatCurve=DataInterface() {propertyValues {Curve {
           defaultValue=0
           keys [{interpMode=RCIM_Cubic, tangentMode=RCTM_User, time=0, value=0}, {interpMode=RCIM_Cubic, tangentMode=RCTM_User, time=1, value=1}]
          }}}
        Scale Curve=&MakeCustomFloatFromBool(
         Bool=$User.IsLargeBody
         True Float=&RandomRangeFloat(Minimum=1, Maximum=9)
         False Float=&RandomRangeFloat(Minimum=1, Maximum=7.5)
        )
       )
      )
     )
    )) {isSetParameters=true, module=SetVariables_#9}
   /SetVariables(Particles.Debugrotation:NiagaraFloat=&Waveform(
     Global Phase=&Add_Float(A=&RandomRangeFloat(Minimum=0, Maximum=1), B=$Engine.Time)
     "Waveform [1]"=::NewEnumerator12
     "[1] Phase Scale"=2
     "[1] Frequency"=0.5
     "[1] Amplitude Min/Max"=<-20 20>
    )) {isSetParameters=true, module=SetVariables_#10}
   /SetVariables(Particles.SpriteRotation:NiagaraFloat=&Add_Float(A=$Particles.Debugrotation, B=$Particles.StartAngle)) {isSetParameters=true, module=SetVariables_#11}
   &ScaleColor(
    Scale Alpha=&Multiply_Float(
     A=&FloatFromCurve(
      FloatCurve=DataInterface() {propertyValues {Curve {
         defaultValue=1
         keys [
          {arriveTangent=13.9519, interpMode=RCIM_Cubic, leaveTangent=13.9519, tangentMode=RCTM_User, time=0, value=0}
          {arriveTangentWeight=0.0469, interpMode=RCIM_Cubic, leaveTangentWeight=0.1864, tangentMode=RCTM_User, time=0.1, value=0.9847}
          {arriveTangentWeight=0.2835, interpMode=RCIM_Cubic, leaveTangentWeight=0.1521, tangentMode=RCTM_User, time=0.55, value=1}
          {arriveTangentWeight=0.676, interpMode=RCIM_Cubic, tangentMode=RCTM_User, time=1, value=0}
         ]
        }}}
      Scale Curve=&MakeCustomFloatFromBool(Bool=$User.IsFrontEnd, True Float=0.8, False Float=1)
     )
     B=&Multiply_Float(
      A=&Waveform(
       Global Phase=&Add_Float(A=$Emitter.Age, B=&RandomRangeFloat(Minimum=0, Maximum=1))
       "Waveform [1]"=::NewEnumerator9
       "[1] Frequency"=1
       "[1] Amplitude Min/Max"=<0.7 1>
      )
      B=&MakeCustomFloatFromBool(
       Bool=&SetBoolByFloatComparison(
        A=&OneMinusFloat(Float=&DotProduct(Vector A=&ReturnCameraVector(Camera Vectors=::NewEnumerator1, Coordinate Space=::Local), Vector B=<0 0 1>))
        B=0.7
       )
       True Float=&FloatFromCurve(
        FloatCurve=DataInterface() {propertyValues {Curve {keys [<at 0.7 1>,<at 0.95 0>]}}}
        CurveIndex=&Abs_Float(Float:NiagaraFloat=&DotProduct(Vector A=&ReturnCameraVector(Camera Vectors=::NewEnumerator1, Coordinate Space=::Local), Vector B=<0 0 1>))
       )
       False Float=1
      )
     )
    )
    Scale RGB=&Add_Vector(A=$System.Color RGB, B=<0.5 0.5 0.5>)
    ScaleA=true
    ScaleRGB=true
   )
  ]
  rendererClasses [&NiagaraSpriteRendererProperties, &NiagaraMeshRendererProperties]
  renderers [
   {rendererClass=&NiagaraSpriteRendererProperties}
   {rendererClass=&NiagaraMeshRendererProperties, rendererIndex=1}
   {rendererClass=&NiagaraSpriteRendererProperties, rendererIndex=2}
  ]
 }
 Static_Icons {
  bEnabled=false
  emitterUpdateScript [
   &EmitterState()
   &SpawnBurst_Instantaneous(Spawn Count=16, Use Spawn Probability=false, UseLoopCountLimit=true)
   /SetVariables(Emitter.Icon Scale Mod:NiagaraFloat=&MakeCustomFloatFromBool(Bool=$User.IsLargeBody, True Float=1.2, False Float=1)) {isSetParameters=true, module=SetVariables_#6}
  ]
  particleSpawnScript [
   &PartitionParticles(
    Normalized Index In Partition Includes One=true
    Number Of Partitions=4
    Particles Per Partition=4
    Partition Mode:ENiagara_PartitionMode_v2=::NewEnumerator1
   )
   &InitializeParticle(
    Color Mode=::NewEnumerator0
    Lifetime=$Emitter.CurrentLoopDuration
    Sprite Size Mode=::NewEnumerator3
    Uniform Sprite Size=&Multiply_Float(A=&RandomRangeFloat(Minimum=7, Maximum=9), B=$Emitter.Icon Scale Mod)
    UsePositionOffset=false
    Write Material Random=true
   )
   &CameraOffset(Camera Offset Amount=-10)
   /SetVariables(Particles.FramePicker:NiagaraInt32=&MergeHelper_IntSwitch(
     0=&SelectIntFromArray(
      Direct Array Index=$Particles.PartitionParticles.Index In Partition
      Int Selection Array=DataInterface() {propertyValues {IntData [10,12,6,14]}}
     )
     1=&SelectIntFromArray(
      Direct Array Index=$Particles.PartitionParticles.Index In Partition
      Int Selection Array=DataInterface() {propertyValues {IntData [8,11,4,14]}}
     )
     2=&SelectIntFromArray(
      Direct Array Index=$Particles.PartitionParticles.Index In Partition
      Int Selection Array=DataInterface() {propertyValues {IntData [1,6,10,12]}}
     )
     3=&SelectIntFromArray(
      Direct Array Index=$Particles.PartitionParticles.Index In Partition
      Int Selection Array=DataInterface() {propertyValues {IntData [10,4,8,15]}}
     )
     4=?
     5=?
     Index=$Particles.PartitionParticles.Partition
     Num=::NewEnumerator2
    )) {isSetParameters=true, module=SetVariables_#7}
   /SetVariables(
    Particles.PivotOffsetPicker:Vector2f=&SelectVector2DFromArray(
     Direct Array Index=$Particles.FramePicker
     Vector 2D Selection Array=DataInterface() {propertyValues {FloatData [
        <0.5 0.95>
        <0.5 0.95>
        <0.4 0.8>
        <0.6 0.8>
        <0.43 0.98>
        <0.57 0.98>
        <0.5 0.95>
        <0.5 0.95>
        <0.5 0.9>
        <0.5 0.9>
        <0.5 0.9>
        <0.5 0.9>
        <0.5 0.93>
        <0.5 0.93>
        <0.5 0.93>
        <0.5 0.93>
        <0.5 0.9>
       ]}}
    )
    Particles.SpriteSize=&Multiply_Vector2D_ByFloat(
     Float=&SelectFloatFromArray(
      Direct Array Index=$Particles.FramePicker
      Float Selection Array=DataInterface() {propertyValues {FloatData [1.2,1.2,0.5,0.5,1.3,1.3,1,1,0.75,0.75,1,1,0.75,0.75,0.75,0.75,1]}}
     )
     Vector2D=$Particles.SpriteSize
    )
    Particles.UVScalePicker:Vector2f=&SelectVector2DFromArray(
     Direct Array Index=$Particles.FramePicker
     Vector 2D Selection Array=DataInterface() {propertyValues {FloatData [<1 1>, <-1 1>, <1 1>, <-1 1>, <1 1>, <-1 1>, <1 1>, <-1 1>, <1 1>, <-1 1>, <1 1>, <-1 1>, <1 1>, <-1 1>, <1 1>, <-1 1>, <1 1>]}}
    )
   ) {isSetParameters=true, module=SetVariables_#8}
   &SubUVAnimation(
    Direct Frame Index:NiagaraFloat=&SelectFloatFromArray(
     Direct Array Index=$Particles.FramePicker
     Float Selection Array=DataInterface() {propertyValues {FloatData [7,7,8,8,9,9,10,10,11,11,12,12,13,13,14,14,15]}}
    )
    "Sprite Renderer "=DataInterface() {propertyValues {SpriteRenderer=&self:Static_Icons_0.NiagaraSpriteRendererProperties_1}}
    SubUV Animation Mode:ENiagara_SubUVLookupModeV2=::NewEnumerator5
   )
  ]
  particleUpdateScript [
   &ParticleState(Kill Particles When Lifetime Has Elapsed=false)
   &ScaleSpriteSize(
    Scale Factor=&MakeVector2D(
     X=&Waveform(
      Global Phase=&Add_Float(A=$Emitter.Age, B=&RandomRangeFloat(Minimum=0, Maximum=2))
      "Waveform [1]"=::NewEnumerator1
      "[1] Frequency"=0.5
      "[1] Amplitude Min/Max"=<0.8 1.1>
     )
     Y=&Waveform(
      Global Phase=&Add_Float(A=$Emitter.Age, B=&RandomRangeFloat(Minimum=0, Maximum=2))
      "Waveform [1]"=::NewEnumerator0
      "[1] Frequency"=0.2
      "[1] Amplitude Min/Max"=<1 1.75>
     )
    )
    Scale Sprite Size Mode=::NewEnumerator2
   )
   &SkeletalMeshLocation(
    Bone / Socket Index=&SelectIntFromArray(Direct Array Index=$Particles.PartitionParticles.Partition, Int Selection Array=DataInterface() {propertyValues {IntData [1,2,7,7]}})
    Bone Sampling Mode=::NewEnumerator4
    MaskSpawnGroup=false
    Sampled Position Offset=&Multiply_Vector(
     A=&MergeHelper_VectorSwitch(
      0=&MakeVector(
       X=&FloatFromCurve(
        FloatCurve=DataInterface() {propertyValues {Curve {defaultValue=1, keys [{tangentMode=RCTM_User, time=0, value=0.1}, {tangentMode=RCTM_User, time=1, value=1}]}}}
        CurveIndex=$Particles.PartitionParticles.Normalized Index In Partition
        Scale Curve=&MakeCustomFloatFromBool(Bool=$User.IsLargeBody, True Float=14, False Float=13)
       )
       Y=&FloatFromCurve(
        FloatCurve=DataInterface() {propertyValues {Curve {keys [<at 0 1>,<at 1 0.1>]}}}
        CurveIndex=$Particles.PartitionParticles.Normalized Index In Partition
        Scale Curve=-2
       )
       Z=&FloatFromCurve(
        FloatCurve=DataInterface() {propertyValues {Curve {keys [<at 0 1>,<at 1 0.1>]}}}
        CurveIndex=$Particles.PartitionParticles.Normalized Index In Partition
        Scale Curve=2
       )
      )
      1=&MakeVector(
       X=&FloatFromCurve(
        FloatCurve=DataInterface() {propertyValues {Curve {defaultValue=1, keys [{tangentMode=RCTM_User, time=0, value=0.1}, {tangentMode=RCTM_User, time=1, value=1}]}}}
        CurveIndex=$Particles.PartitionParticles.Normalized Index In Partition
        Scale Curve=&MakeCustomFloatFromBool(Bool=$User.IsLargeBody, True Float=-14, False Float=-13)
       )
       Y=&FloatFromCurve(
        FloatCurve=DataInterface() {propertyValues {Curve {keys [<at 0 1>,<at 1 0.1>]}}}
        CurveIndex=$Particles.PartitionParticles.Normalized Index In Partition
        Scale Curve=2
       )
       Z=&FloatFromCurve(
        FloatCurve=DataInterface() {propertyValues {Curve {keys [<at 0 1>,<at 1 0.1>]}}}
        CurveIndex=$Particles.PartitionParticles.Normalized Index In Partition
        Scale Curve=-2
       )
      )
      2=&MakeVector(
       X=&FloatFromCurve(
        FloatCurve=DataInterface() {propertyValues {Curve {defaultValue=1, keys [{tangentMode=RCTM_User, time=0, value=1}, {tangentMode=RCTM_User, time=1, value=1}]}}}
        CurveIndex=$Particles.PartitionParticles.Normalized Index In Partition
        Scale Curve=3
       )
       Y=&FloatFromCurve(
        FloatCurve=DataInterface() {propertyValues {Curve {keys [<at 0 0>,<at 1 0>]}}}
        CurveIndex=$Particles.PartitionParticles.Normalized Index In Partition
        Scale Curve=3
       )
       Z=&FloatFromCurve(
        FloatCurve=DataInterface() {propertyValues {Curve {keys [<at 0 0.45>,<at 1 1>]}}}
        CurveIndex=$Particles.PartitionParticles.Normalized Index In Partition
        Scale Curve=&MakeCustomFloatFromBool(Bool=$User.IsLargeBody, True Float=20, False Float=14)
       )
      )
      3=&MakeVector(
       X=&FloatFromCurve(
        FloatCurve=DataInterface() {propertyValues {Curve {defaultValue=1, keys [{tangentMode=RCTM_User, time=0, value=1}, {tangentMode=RCTM_User, time=1, value=1}]}}}
        CurveIndex=$Particles.PartitionParticles.Normalized Index In Partition
        Scale Curve=3
       )
       Y=&FloatFromCurve(
        FloatCurve=DataInterface() {propertyValues {Curve {keys [<at 0 0>,<at 1 0>]}}}
        CurveIndex=$Particles.PartitionParticles.Normalized Index In Partition
        Scale Curve=3
       )
       Z=&FloatFromCurve(
        FloatCurve=DataInterface() {propertyValues {Curve {keys [<at 0 -0.45>,<at 1 -1>]}}}
        CurveIndex=$Particles.PartitionParticles.Normalized Index In Partition
        Scale Curve=&MakeCustomFloatFromBool(Bool=$User.IsLargeBody, True Float=20, False Float=14)
       )
      )
      Index=$Particles.PartitionParticles.Partition
      Num=::NewEnumerator2
     )
     B=&MakeCustomVectorFromBool(Bool=$User.IsLargeBody, True Vector=<1 1 1>, False Vector=<1 1 1>)
    )
    Skeletal Mesh=$System.Skeletal Mesh
    Use Sampled Mesh Attribute as Ribbon ID=false
   )
   &SolveForcesAndVelocity(Clamp Velocity=false, Limit Acceleration=false)
   /SetVariables(Particles.PositionOffset:NiagaraPosition=&AddVectorToPosition(
     Position=$Particles.Position
     Vector=&Multiply_Vector(
      A=&Add_Vector(
       A=&MergeHelper_VectorSwitch(
        0=<0 0 0>
        1=<0 0 0>
        2=&MakeVector(
         X=0
         Y=0
         Z=&FloatFromCurve(
          FloatCurve=DataInterface() {propertyValues {Curve {keys [<at 0 1.5 RCIM_Cubic>,<at 1 1 RCIM_Cubic>]}}}
          CurveIndex=$Particles.PartitionParticles.Normalized Index In Partition
          Scale Curve=&Multiply_Float(
           A=&MakeCustomFloatFromBool(Bool=$User.IsLargeBody, True Float=2, False Float=1)
           B=&MakeCustomFloatFromBool(Bool=$User.IsFrontEnd, True Float=1.5, False Float=1)
          )
         )
        )
        3=&MakeVector(
         X=0
         Y=0
         Z=&FloatFromCurve(
          FloatCurve=DataInterface() {propertyValues {Curve {keys [<at 0 1.5 RCIM_Cubic>,<at 1 1 RCIM_Cubic>]}}}
          CurveIndex=$Particles.PartitionParticles.Normalized Index In Partition
          Scale Curve=&Multiply_Float(
           A=&MakeCustomFloatFromBool(Bool=$User.IsLargeBody, True Float=2, False Float=1)
           B=&MakeCustomFloatFromBool(Bool=$User.IsFrontEnd, True Float=1.5, False Float=1)
          )
         )
        )
        Index=$Particles.PartitionParticles.Partition
        Num=::NewEnumerator2
       )
       B=&MakeCustomVectorFromBool(
        Bool=$User.IsFrontEnd
        True Vector=&MakeCustomVectorFromBool(Bool=$User.IsLargeBody, True Vector=<0 0 2.5>, False Vector=<0 0 1.5>)
        False Vector=<0 0 2>
       )
      )
      B=<1 1 1>
     )
    )) {isSetParameters=true, module=SetVariables_#9}
   /SetVariables(Particles.RotationWaveform:NiagaraFloat=&Waveform(
     Global Phase=&Add_Float(A=$Particles.PartitionParticles.Normalized Index In Partition, B=$Engine.Time)
     "Waveform [1]"=::NewEnumerator12
     "[1] Frequency"=0.5
     "[1] Amplitude Min/Max"=<-20 20>
    )) {isSetParameters=true, module=SetVariables_#10}
   /SetVariables(Particles.SpriteRotation:NiagaraFloat=&Add_Float(A=&Multiply_Float(A=$Particles.RotationWaveform, B=1), B=$Particles.StartAngle)) {isSetParameters=true, module=SetVariables_#11}
   &ScaleColor(
    Scale Alpha=&Multiply_Float(
     A=&Multiply_Float(
      A=&Waveform(
       Global Phase=&Add_Float(A=$Emitter.Age, B=&RandomRangeFloat(Minimum=0, Maximum=1))
       "Waveform [1]"=::NewEnumerator9
       "[1] Frequency"=1
       "[1] Amplitude Min/Max"=<0.5 1>
      )
      B=&MakeCustomFloatFromBool(
       Bool=&SetBoolByFloatComparison(
        A=&OneMinusFloat(Float=&DotProduct(Vector A=&ReturnCameraVector(Camera Vectors=::NewEnumerator1, Coordinate Space=::Local), Vector B=<0 0 1>))
        B=0.7
       )
       True Float=&FloatFromCurve(
        FloatCurve=DataInterface() {propertyValues {Curve {keys [<at 0.7 1>,<at 0.95 0>]}}}
        CurveIndex=&Abs_Float(Float:NiagaraFloat=&DotProduct(Vector A=&ReturnCameraVector(Camera Vectors=::NewEnumerator1, Coordinate Space=::Local), Vector B=<0 0 1>))
       )
       False Float=1
      )
     )
     B=&MakeCustomFloatFromBool(Bool=$User.IsFrontEnd, True Float=0.8, False Float=1)
    )
    Scale RGB=&Add_Vector(A=$System.Color RGB, B=<0.5 0.5 0.5>)
    ScaleA=true
    ScaleRGB=true
   )
  ]
  rendererClasses [&NiagaraSpriteRendererProperties, &NiagaraMeshRendererProperties]
  renderers [
   {rendererClass=&NiagaraSpriteRendererProperties}
   {rendererClass=&NiagaraMeshRendererProperties, rendererIndex=1}
   {rendererClass=&NiagaraSpriteRendererProperties, rendererIndex=2}
  ]
 }
 Stem_Ribbons {
  emitterSpawnScript [/SetVariables(Emitter.Number Of Ribbons:NiagaraInt32=$System.NumberOfRibbons, Emitter.Particles Per Ribbon:NiagaraInt32=$System.ParticlesPerRibbon) {isSetParameters=true}]
  emitterUpdateScript [
   &EmitterState()
   &SpawnBurst_Instantaneous(
    Spawn Count=&Multiply_Int(A=$Emitter.Initial.Number Of Ribbons, B=$Emitter.Particles Per Ribbon)
    Use Spawn Probability=false
    UseLoopCountLimit=true
   )
  ]
  particleSpawnScript [
   &PartitionParticles(
    Number Of Partitions=$Emitter.Number Of Ribbons
    Particles Per Partition:NiagaraInt32=$Emitter.Particles Per Ribbon
    Partition Mode:ENiagara_PartitionMode_v2=::NewEnumerator1
    Write Partition Index to Ribbon Link Order:ENiagara_AttributeSamplingApplyOutput=::NewEnumerator0
    Write Partition to Ribbon ID=::NewEnumerator0
   )
   &InitializeParticle(
    Color=&MakeLinearColorFromVectorAndFloat(
     "Vector (RGB)"=&MergeHelper_VectorSwitch(
      0=&MakeVectorFromLinearColorRGB(Color=&ColorFromCurve(
        ColorCurve=DataInterface() {propertyValues {
          AlphaCurve {keys [<at 0 1 RCIM_Cubic>]}
          BlueCurve {keys [<at 0.0633 0.0086>,<at 0.2483 0>,<at 0.5823 0.0107 RCIM_Cubic>,<at 1 0 RCIM_Cubic>]}
          GreenCurve {keys [
            <at 0.0633 0.0356>
            <at 0.2483 0.1713>
            {arriveTangent=-0.1781, interpMode=RCIM_Cubic, leaveTangent=-0.1781, time=0.5823, value=0.0445}
            <at 1 0.0374 RCIM_Cubic>
           ]}
          RedCurve {keys [
            <at 0.0633 1>
            <at 0.2483 1>
            {arriveTangent=0.6652, interpMode=RCIM_Cubic, leaveTangent=0.6652, time=0.5823, value=1.25}
            <at 1 1.5 RCIM_Cubic>
           ]}
         }}
        CurveIndex=$Particles.PartitionParticles.Normalized Index In Partition
       ))
      1=&MakeVectorFromLinearColorRGB(Color=&ColorFromCurve(
        ColorCurve=DataInterface() {propertyValues {
          AlphaCurve {keys [<at 0 1 RCIM_Cubic>]}
          BlueCurve {keys [<at 0.2206 0.0532 RCIM_Cubic>,<at 1 0.4154 RCIM_Cubic>]}
          GreenCurve {keys [<at 0.2206 0 RCIM_Cubic>,<at 1 0.0313 RCIM_Cubic>]}
          RedCurve {keys [<at 0.2206 1 RCIM_Cubic>,<at 1 1.2 RCIM_Cubic>]}
         }}
        CurveIndex=$Particles.PartitionParticles.Normalized Index In Partition
       ))
      2=&MakeVectorFromLinearColorRGB(Color=&ColorFromCurve(
        ColorCurve=DataInterface() {propertyValues {
          AlphaCurve {keys [<at 0 1 RCIM_Cubic>]}
          BlueCurve {keys [<at 0.0633 0.0086>,<at 0.2483 0>,<at 0.5823 0.0107 RCIM_Cubic>,<at 1 0 RCIM_Cubic>]}
          GreenCurve {keys [
            <at 0.0633 0.0356>
            <at 0.2483 0.1713>
            {arriveTangent=-0.1781, interpMode=RCIM_Cubic, leaveTangent=-0.1781, time=0.5823, value=0.0445}
            <at 1 0.0374 RCIM_Cubic>
           ]}
          RedCurve {keys [
            <at 0.0633 1>
            <at 0.2483 1>
            {arriveTangent=0.6652, interpMode=RCIM_Cubic, leaveTangent=0.6652, time=0.5823, value=1.25}
            <at 1 1.5 RCIM_Cubic>
           ]}
         }}
        CurveIndex=$Particles.PartitionParticles.Normalized Index In Partition
       ))
      3=&MakeVectorFromLinearColorRGB(Color=&ColorFromCurve(
        ColorCurve=DataInterface() {propertyValues {
          AlphaCurve {keys [<at 0 1 RCIM_Cubic>]}
          BlueCurve {keys [<at 0.0633 0.0086>,<at 0.2483 0>,<at 0.5823 0.0107 RCIM_Cubic>,<at 1 0 RCIM_Cubic>]}
          GreenCurve {keys [
            <at 0.0633 0.0356>
            <at 0.2483 0.1713>
            {arriveTangent=-0.1781, interpMode=RCIM_Cubic, leaveTangent=-0.1781, time=0.5823, value=0.0445}
            <at 1 0.0374 RCIM_Cubic>
           ]}
          RedCurve {keys [
            <at 0.0633 1>
            <at 0.2483 1>
            {arriveTangent=0.6652, interpMode=RCIM_Cubic, leaveTangent=0.6652, time=0.5823, value=1.25}
            <at 1 1.5 RCIM_Cubic>
           ]}
         }}
        CurveIndex=$Particles.PartitionParticles.Normalized Index In Partition
       ))
      4=&MakeVectorFromLinearColorRGB(Color=&ColorFromCurve(
        ColorCurve=DataInterface() {propertyValues {
          AlphaCurve {keys [<at 0 1 RCIM_Cubic>]}
          BlueCurve {keys [<at 0.2011 0.0532 RCIM_Cubic>,<at 1 0.4154 RCIM_Cubic>]}
          GreenCurve {keys [<at 0.2011 0 RCIM_Cubic>,<at 1 0.0312 RCIM_Cubic>]}
          RedCurve {keys [<at 0.2011 1 RCIM_Cubic>,<at 1 1.2 RCIM_Cubic>]}
         }}
        CurveIndex=$Particles.PartitionParticles.Normalized Index In Partition
       ))
      5=&MakeVectorFromLinearColorRGB(Color=&ColorFromCurve(
        ColorCurve=DataInterface() {propertyValues {
          AlphaCurve {keys [<at 0 1 RCIM_Cubic>]}
          BlueCurve {keys [<at 0.0633 0.0086>,<at 0.2483 0>,<at 0.5823 0.0107 RCIM_Cubic>,<at 1 0 RCIM_Cubic>]}
          GreenCurve {keys [
            <at 0.0633 0.0356>
            <at 0.2483 0.1713>
            {arriveTangent=-0.1781, interpMode=RCIM_Cubic, leaveTangent=-0.1781, time=0.5823, value=0.0445}
            <at 1 0.0374 RCIM_Cubic>
           ]}
          RedCurve {keys [
            <at 0.0633 1>
            <at 0.2483 1>
            {arriveTangent=0.6652, interpMode=RCIM_Cubic, leaveTangent=0.6652, time=0.5823, value=1.25}
            <at 1 1.5 RCIM_Cubic>
           ]}
         }}
        CurveIndex=$Particles.PartitionParticles.Normalized Index In Partition
       ))
      Index=$Particles.PartitionParticles.Partition
      Num=::NewEnumerator4
     )
     "Float (Alpha)"=1
    )
    Sprite Size Mode=::NewEnumerator3
    UsePositionOffset=false
    Write Material Random=true
   )
  ]
  particleUpdateScript [
   &ParticleState(Kill Particles When Lifetime Has Elapsed=false)
   &Timeline(
    Additional Output:ENiagara_TimelineOutput=::NewEnumerator1
    Additional Output Mode:ENiagara_TimelineOutputMode=::NewEnumerator1
    Float Curve=DataInterface() {propertyValues {Curve {keys [
        {interpMode=RCIM_Cubic, tangentMode=RCTM_User, time=0, value=0}
        {arriveTangentWeight=0.0333, interpMode=RCIM_Cubic, leaveTangentWeight=0.2333, tangentMode=RCTM_User, time=0.1, value=1}
        {arriveTangentWeight=0.3488, interpMode=RCIM_Cubic, leaveTangentWeight=0.0997, tangentMode=RCTM_User, time=0.8, value=1}
        {tangentMode=RCTM_User, time=1, value=0}
       ]}}}
   )
   &SkeletalMeshLocation(
    Bone / Socket Index=7
    Bone Sampling Mode=::NewEnumerator4
    "Inherit Velocity (Scale)":NiagaraFloat=&FloatFromCurve(
     FloatCurve=DataInterface() {propertyValues {Curve {defaultValue=1, keys [{tangentMode=RCTM_User, time=0, value=0}, {tangentMode=RCTM_User, time=1, value=1}]}}}
     CurveIndex=$Particles.PartitionParticles.Normalized Index In Partition
     Scale Curve=-0.2
    )
    Inherited Velocity Speed Limit=50
    MaskSpawnGroup=false
    Orientation Sampling=::NewEnumerator1
    Sampled Position Offset=&AddVectorToPosition(
     Position=&MergeHelper_VectorSwitch(
      0=&MakeVector(
       X=0
       Y=&FloatFromCurve(
        FloatCurve=DataInterface() {propertyValues {Curve {keys [{arriveTangent=-3.8443, interpMode=RCIM_Cubic, leaveTangent=-3.8443, tangentMode=RCTM_User, time=0, value=0}, <at 1 -2 RCIM_Cubic>]}}}
        CurveIndex=$Particles.PartitionParticles.Normalized Index In Partition
        Scale Curve=2
       )
       Z=&FloatFromCurve(
        FloatCurve=DataInterface() {propertyValues {Curve {keys [
            {arriveTangent=-132.2883, interpMode=RCIM_Cubic, leaveTangent=-132.2883, tangentMode=RCTM_User, time=0, value=-10}
            {arriveTangent=-19.0337, interpMode=RCIM_Cubic, leaveTangent=-19.0337, tangentMode=RCTM_User, time=0.15, value=-18}
            {arriveTangent=5.7208, interpMode=RCIM_Cubic, leaveTangent=5.7208, tangentMode=RCTM_User, time=0.9885, value=-24.308}
           ]}}}
        CurveIndex=$Particles.PartitionParticles.Normalized Index In Partition
       )
      )
      1=&MakeVector(
       X=0
       Y=&FloatFromCurve(
        FloatCurve=DataInterface() {propertyValues {Curve {keys [{arriveTangent=-9.5531, interpMode=RCIM_Cubic, leaveTangent=-9.5531, tangentMode=RCTM_User, time=0, value=0}, <at 1 -5 RCIM_Cubic>]}}}
        CurveIndex=$Particles.PartitionParticles.Normalized Index In Partition
       )
       Z=&FloatFromCurve(
        FloatCurve=DataInterface() {propertyValues {Curve {keys [
            {arriveTangent=-31.3133, interpMode=RCIM_Cubic, leaveTangent=-31.3133, tangentMode=RCTM_User, time=0, value=-8}
            {arriveTangent=-15.5942, interpMode=RCIM_Cubic, leaveTangent=-15.5942, tangentMode=RCTM_User, time=0.15, value=-12}
            {arriveTangent=5.1364, interpMode=RCIM_Cubic, leaveTangent=5.1364, tangentMode=RCTM_User, time=1, value=-18}
           ]}}}
        CurveIndex=$Particles.PartitionParticles.Normalized Index In Partition
       )
      )
      2=&MakeVector(
       X=0
       Y=&FloatFromCurve(
        FloatCurve=DataInterface() {propertyValues {Curve {keys [
            {arriveTangent=-14.62, interpMode=RCIM_Cubic, leaveTangent=-14.62, leaveTangentWeight=3.2018, tangentMode=RCTM_User, time=0, value=0}
            <at 1 -5 RCIM_Cubic>
           ]}}}
        CurveIndex=$Particles.PartitionParticles.Normalized Index In Partition
        Scale Curve=1.5
       )
       Z=&FloatFromCurve(
        FloatCurve=DataInterface() {propertyValues {Curve {keys [
            {arriveTangent=-46.8682, interpMode=RCIM_Cubic, leaveTangent=-46.8682, tangentMode=RCTM_User, time=0, value=0}
            {arriveTangent=-13.7244, interpMode=RCIM_Cubic, leaveTangent=-13.7244, tangentMode=RCTM_User, time=0.15, value=-5}
            {arriveTangentWeight=4.6683, interpMode=RCIM_Cubic, tangentMode=RCTM_User, time=1, value=-10}
           ]}}}
        CurveIndex=$Particles.PartitionParticles.Normalized Index In Partition
       )
      )
      3=&MakeVector(
       X=0
       Y=&FloatFromCurve(
        FloatCurve=DataInterface() {propertyValues {Curve {keys [
            {arriveTangent=-3.9837, interpMode=RCIM_Cubic, leaveTangent=-3.9837, leaveTangentWeight=3.8607, tangentMode=RCTM_User, time=0, value=0}
            {arriveTangentWeight=2.0105, interpMode=RCIM_Cubic, tangentMode=RCTM_User, time=1, value=-2}
           ]}}}
        CurveIndex=$Particles.PartitionParticles.Normalized Index In Partition
        Scale Curve=1.2
       )
       Z=&FloatFromCurve(
        FloatCurve=DataInterface() {propertyValues {Curve {keys [
            {arriveTangent=53.6775, interpMode=RCIM_Cubic, leaveTangent=53.6775, tangentMode=RCTM_User, time=0, value=0}
            {arriveTangent=15.1979, interpMode=RCIM_Cubic, leaveTangent=15.1979, tangentMode=RCTM_User, time=0.15, value=5}
            {arriveTangentWeight=2.1616, interpMode=RCIM_Cubic, tangentMode=RCTM_User, time=1, value=10}
           ]}}}
        CurveIndex=$Particles.PartitionParticles.Normalized Index In Partition
       )
      )
      4=&MakeVector(
       X=0
       Y=&FloatFromCurve(
        FloatCurve=DataInterface() {propertyValues {Curve {keys [
            {arriveTangent=-8.1052, interpMode=RCIM_Cubic, leaveTangent=-8.1052, leaveTangentWeight=3.8607, tangentMode=RCTM_User, time=0, value=0}
            <at 1 -4.45 RCIM_Cubic>
           ]}}}
        CurveIndex=$Particles.PartitionParticles.Normalized Index In Partition
        Scale Curve=2
       )
       Z=&FloatFromCurve(
        FloatCurve=DataInterface() {propertyValues {Curve {keys [
            {arriveTangent=75.6204, interpMode=RCIM_Cubic, leaveTangent=75.6204, tangentMode=RCTM_User, time=0, value=5}
            {arriveTangent=18.4907, interpMode=RCIM_Cubic, leaveTangent=18.4907, tangentMode=RCTM_User, time=0.15, value=12}
            {arriveTangent=-10.6114, arriveTangentWeight=4.6683, interpMode=RCIM_Cubic, leaveTangent=-10.6114, tangentMode=RCTM_User, time=1, value=18}
           ]}}}
        CurveIndex=$Particles.PartitionParticles.Normalized Index In Partition
       )
      )
      5=&MakeVector(
       X=0
       Y=&FloatFromCurve(
        FloatCurve=DataInterface() {propertyValues {Curve {keys [
            {arriveTangent=-6.7419, interpMode=RCIM_Cubic, leaveTangent=-6.7419, leaveTangentWeight=3.8607, tangentMode=RCTM_User, time=0, value=0}
            <at 1 -2 RCIM_Cubic>
           ]}}}
        CurveIndex=$Particles.PartitionParticles.Normalized Index In Partition
        Scale Curve=1.5
       )
       Z=&FloatFromCurve(
        FloatCurve=DataInterface() {propertyValues {Curve {keys [
            {arriveTangent=108.6479, interpMode=RCIM_Cubic, leaveTangent=108.6479, tangentMode=RCTM_User, time=0, value=10}
            {arriveTangent=20.2957, interpMode=RCIM_Cubic, leaveTangent=20.2957, tangentMode=RCTM_User, time=0.15, value=18}
            {arriveTangent=-0.7118, arriveTangentWeight=4.6683, interpMode=RCIM_Cubic, leaveTangent=-0.7118, tangentMode=RCTM_User, time=1, value=25}
           ]}}}
        CurveIndex=$Particles.PartitionParticles.Normalized Index In Partition
       )
      )
      Index=$Particles.PartitionParticles.Partition
      Num=::NewEnumerator4
     )
     Vector=&MakeCustomVectorFromBool(Bool=$User.IsLargeBody, True Vector=<3 -3.5 0>, False Vector=<0 -3 0>)
    )
    Skeletal Mesh=$System.Skeletal Mesh
    Use Sampled Mesh Attribute as Ribbon ID=false
    Velocity Sampling=::NewEnumerator0
    VelocitySpeedLimit=true
   )
   &Drag(Drag=2, Use Linear Drag=true, Use Rotational Drag=true)
   &SolveForcesAndVelocity(Clamp Velocity=false, Limit Acceleration=false)
   /SetVariables(Particles.WidthCurve:NiagaraFloat=&FloatFromCurve(
     FloatCurve=DataInterface() {propertyValues {Curve {
        defaultValue=1
        keys [
         {arriveTangent=5.5397, interpMode=RCIM_Cubic, leaveTangent=5.5397, tangentMode=RCTM_User, time=0, value=0.2046}
         {arriveTangent=0.2928, interpMode=RCIM_Cubic, leaveTangent=0.2928, tangentMode=RCTM_User, time=0.1651, value=0.5718}
         {arriveTangent=-0.767, interpMode=RCIM_Cubic, leaveTangent=-0.767, tangentMode=RCTM_User, time=0.4584, value=0.4085}
         {arriveTangent=-0.4427, interpMode=RCIM_Cubic, leaveTangent=-0.4427, tangentMode=RCTM_User, time=0.7757, value=0.2226}
         {arriveTangent=0.0177, interpMode=RCIM_Cubic, leaveTangent=0.0177, tangentMode=RCTM_User, time=1, value=0}
        ]
       }}}
     CurveIndex=$Particles.PartitionParticles.Normalized Index In Partition
     Scale Curve=2.5
    )) {isSetParameters=true}
   &ScaleColor(
    Scale Alpha=&FloatFromCurve(
     FloatCurve=DataInterface() {propertyValues {Curve {defaultValue=1, keys [<at 0 0 RCIM_Cubic>, {tangentMode=RCTM_User, time=0.4, value=1}]}}}
     CurveIndex=$Particles.PartitionParticles.Normalized Index In Partition
    )
    Scale RGB=&Multiply_Vector(A=&MakeVectorFromLinearColorRGB(Color=<rgba 1 1 1 1>), B=<1 1 1>)
    ScaleA=true
    ScaleRGB=true
   )
   /SetVariables(Particles.Offset Position:NiagaraPosition=&AddVectorToPosition(
     Position=$Particles.Position
     Vector=&Add_Vector(
      A=&Multiply_VectorByFloat(
       Vector=&MakeVector(
        X=&Waveform(
         Global Phase=&Add_Float(
          A=$Particles.PartitionParticles.NormalizedPartition
          B=&Add_Float(A=$Engine.Time, B=$Particles.PartitionParticles.Normalized Index In Partition)
         )
         "Waveform [1]"=::NewEnumerator1
         "[1] Phase Scale"=0.75
         "[1] Frequency"=1
         "[1] Amplitude Min/Max"=<-5 10>
        )
        Y=&Waveform(
         Global Phase=&Add_Float(A=$Engine.Time, B=$Particles.PartitionParticles.Normalized Index In Partition)
         "Waveform [1]"=::NewEnumerator0
         "[1] Frequency"=0.5
         "[1] Amplitude Min/Max"=<-5 10>
        )
        Z=&Waveform(
         Global Phase=&Add_Float(A=$Engine.Time, B=$Particles.PartitionParticles.Normalized Index In Partition)
         "Waveform [1]"=::NewEnumerator0
         "[1] Frequency"=1
         "[1] Amplitude Min/Max"=<-5 7>
        )
       )
       Float=&FloatFromCurve(
        FloatCurve=DataInterface() {propertyValues {Curve {
           defaultValue=1
           keys [
            {arriveTangent=0.6545, interpMode=RCIM_Cubic, leaveTangent=0.6545, tangentMode=RCTM_User, time=0.1, value=0}
            {arriveTangent=0.2587, interpMode=RCIM_Cubic, leaveTangent=0.2587, tangentMode=RCTM_User, time=0.5, value=0.4498}
            {arriveTangent=3.3325, interpMode=RCIM_Cubic, leaveTangent=3.3325, tangentMode=RCTM_User, time=0.9009, value=0.8049}
           ]
          }}}
        CurveIndex=$Particles.PartitionParticles.Normalized Index In Partition
        Scale Curve=0.19
       )
      )
      B=&Multiply_Vector(
       A=&MergeHelper_VectorSwitch(
        0=&MakeVector(
         X=&FloatFromCurve(
          FloatCurve=DataInterface() {propertyValues {Curve {
             defaultValue=0
             keys [
              <at 0 0 RCIM_Cubic>
              {arriveTangentWeight=0.0794, interpMode=RCIM_Cubic, leaveTangentWeight=0.0983, tangentMode=RCTM_User, time=0.2217, value=0}
              {arriveTangent=0.2712, interpMode=RCIM_Cubic, leaveTangent=0.2712, time=0.4964, value=-0.1941}
              {arriveTangent=0.3854, interpMode=RCIM_Cubic, leaveTangent=0.3854, time=0.8073, value=0.1588}
              <at 1 0 RCIM_Cubic>
             ]
            }}}
          CurveIndex=$Particles.PartitionParticles.Normalized Index In Partition
         )
         Y=&FloatFromCurve(
          FloatCurve=DataInterface() {propertyValues {Curve {
             defaultValue=1
             keys [
              <at 0 0 RCIM_Cubic>
              {arriveTangentWeight=0.0716, interpMode=RCIM_Cubic, leaveTangentWeight=0.1061, tangentMode=RCTM_User, time=0.2, value=0}
              {arriveTangent=0.2615, interpMode=RCIM_Cubic, leaveTangent=0.2615, time=0.4964, value=-0.1941}
              {arriveTangent=0.3854, interpMode=RCIM_Cubic, leaveTangent=0.3854, time=0.8073, value=0.1588}
              <at 1 0 RCIM_Cubic>
             ]
            }}}
          CurveIndex=$Particles.PartitionParticles.Normalized Index In Partition
         )
         Z=&FloatFromCurve(
          FloatCurve=DataInterface() {propertyValues {Curve {defaultValue=1, keys [{tangentMode=RCTM_User, time=0, value=0}, <at 0.85 0.95>, {tangentMode=RCTM_User, time=1, value=1}]}}}
          CurveIndex=$Particles.PartitionParticles.Normalized Index In Partition
          Scale Curve=21
         )
        )
        1=&MakeVector(
         X=&FloatFromCurve(
          FloatCurve=DataInterface() {propertyValues {Curve {
             defaultValue=1
             keys [
              {interpMode=RCIM_Cubic, leaveTangentWeight=0.0692, tangentMode=RCTM_User, time=0, value=0}
              {arriveTangentWeight=0.0692, interpMode=RCIM_Cubic, leaveTangentWeight=0.0952, tangentMode=RCTM_User, time=0.2499, value=0}
              {arriveTangentWeight=0.0952, interpMode=RCIM_Cubic, leaveTangentWeight=0.1033, tangentMode=RCTM_User, time=0.5177, value=0.25}
              {arriveTangentWeight=0.1033, interpMode=RCIM_Cubic, leaveTangentWeight=0.0656, tangentMode=RCTM_User, time=0.7509, value=-0.1912}
              {arriveTangentWeight=0.0656, interpMode=RCIM_Cubic, tangentMode=RCTM_User, time=1, value=0}
             ]
            }}}
          CurveIndex=$Particles.PartitionParticles.Normalized Index In Partition
          Scale Curve=3
         )
         Y=&FloatFromCurve(
          FloatCurve=DataInterface() {propertyValues {Curve {
             defaultValue=1
             keys [
              {interpMode=RCIM_Cubic, leaveTangentWeight=0.0667, tangentMode=RCTM_User, time=0, value=0}
              {arriveTangentWeight=0.0667, interpMode=RCIM_Cubic, leaveTangentWeight=0.0988, tangentMode=RCTM_User, time=0.2, value=0}
              {arriveTangentWeight=0.0988, interpMode=RCIM_Cubic, leaveTangentWeight=0.1036, tangentMode=RCTM_User, time=0.5999, value=-0.152}
              {arriveTangentWeight=0.1036, interpMode=RCIM_Cubic, leaveTangentWeight=0.0642, tangentMode=RCTM_User, time=0.7932, value=0.2059}
              {arriveTangentWeight=0.0642, interpMode=RCIM_Cubic, tangentMode=RCTM_User, time=1, value=0}
             ]
            }}}
          CurveIndex=$Particles.PartitionParticles.Normalized Index In Partition
          Scale Curve=3
         )
         Z=&FloatFromCurve(
          FloatCurve=DataInterface() {propertyValues {Curve {defaultValue=1, keys [{tangentMode=RCTM_User, time=0, value=0}, <at 0.85 0.95>, {tangentMode=RCTM_User, time=1, value=1}]}}}
          CurveIndex=$Particles.PartitionParticles.Normalized Index In Partition
          Scale Curve=40
         )
        )
        2=&MakeVector(
         X=&FloatFromCurve(
          FloatCurve=DataInterface() {propertyValues {Curve {
             defaultValue=1
             keys [
              {interpMode=RCIM_Cubic, leaveTangentWeight=0.0692, tangentMode=RCTM_User, time=0, value=0}
              {arriveTangentWeight=0.0692, interpMode=RCIM_Cubic, leaveTangentWeight=0.0952, tangentMode=RCTM_User, time=0.2077, value=0}
              {arriveTangentWeight=0.0952, interpMode=RCIM_Cubic, leaveTangentWeight=0.1033, tangentMode=RCTM_User, time=0.5062, value=0.1029}
              {arriveTangentWeight=0.1033, interpMode=RCIM_Cubic, leaveTangentWeight=0.0656, tangentMode=RCTM_User, time=0.7803, value=-0.1961}
              {arriveTangentWeight=0.0656, interpMode=RCIM_Cubic, tangentMode=RCTM_User, time=1, value=0}
             ]
            }}}
          CurveIndex=$Particles.PartitionParticles.Normalized Index In Partition
          Scale Curve=3
         )
         Y=&FloatFromCurve(
          FloatCurve=DataInterface() {propertyValues {Curve {
             defaultValue=1
             keys [
              {interpMode=RCIM_Cubic, leaveTangentWeight=0.0667, tangentMode=RCTM_User, time=0, value=0}
              {arriveTangentWeight=0.0667, interpMode=RCIM_Cubic, leaveTangentWeight=0.0988, tangentMode=RCTM_User, time=0.2, value=0}
              {arriveTangentWeight=0.0988, interpMode=RCIM_Cubic, leaveTangentWeight=0.1036, tangentMode=RCTM_User, time=0.5539, value=-0.1373}
              {arriveTangentWeight=0.1036, interpMode=RCIM_Cubic, leaveTangentWeight=0.0642, tangentMode=RCTM_User, time=0.8213, value=0.1716}
              {arriveTangentWeight=0.0642, interpMode=RCIM_Cubic, tangentMode=RCTM_User, time=1, value=0}
             ]
            }}}
          CurveIndex=$Particles.PartitionParticles.Normalized Index In Partition
          Scale Curve=3
         )
         Z=&FloatFromCurve(
          FloatCurve=DataInterface() {propertyValues {Curve {defaultValue=1, keys [{tangentMode=RCTM_User, time=0, value=0}, <at 0.85 0.95>, {tangentMode=RCTM_User, time=1, value=1}]}}}
          CurveIndex=$Particles.PartitionParticles.Normalized Index In Partition
          Scale Curve=35
         )
        )
        3=&MakeVector(
         X=&FloatFromCurve(
          FloatCurve=DataInterface() {propertyValues {Curve {
             defaultValue=1
             keys [
              {interpMode=RCIM_Cubic, leaveTangentWeight=0.0692, tangentMode=RCTM_User, time=0, value=0}
              {arriveTangentWeight=0.0692, interpMode=RCIM_Cubic, leaveTangentWeight=0.0952, tangentMode=RCTM_User, time=0.2077, value=0}
              {arriveTangentWeight=0.0952, interpMode=RCIM_Cubic, leaveTangentWeight=0.1033, tangentMode=RCTM_User, time=0.4934, value=-0.2108}
              {
               arriveTangent=-0.5432
               arriveTangentWeight=0.1033
               interpMode=RCIM_Cubic
               leaveTangent=-0.5432
               leaveTangentWeight=0.0656
               tangentMode=RCTM_User
               time=0.7675
               value=0.1127
              }
              {arriveTangentWeight=0.0656, interpMode=RCIM_Cubic, tangentMode=RCTM_User, time=1, value=0}
             ]
            }}}
          CurveIndex=$Particles.PartitionParticles.Normalized Index In Partition
          Scale Curve=3
         )
         Y=&FloatFromCurve(
          FloatCurve=DataInterface() {propertyValues {Curve {
             defaultValue=1
             keys [
              {interpMode=RCIM_Cubic, leaveTangentWeight=0.0667, tangentMode=RCTM_User, time=0, value=0}
              {arriveTangentWeight=0.0667, interpMode=RCIM_Cubic, leaveTangentWeight=0.0988, tangentMode=RCTM_User, time=0.2, value=0}
              {arriveTangentWeight=0.0988, interpMode=RCIM_Cubic, leaveTangentWeight=0.1036, tangentMode=RCTM_User, time=0.4947, value=0.1696}
              {arriveTangentWeight=0.1036, interpMode=RCIM_Cubic, leaveTangentWeight=0.0642, tangentMode=RCTM_User, time=0.8379, value=-0.124}
              {arriveTangentWeight=0.0642, interpMode=RCIM_Cubic, tangentMode=RCTM_User, time=1, value=0}
             ]
            }}}
          CurveIndex=$Particles.PartitionParticles.Normalized Index In Partition
          Scale Curve=3
         )
         Z=&FloatFromCurve(
          FloatCurve=DataInterface() {propertyValues {Curve {defaultValue=1, keys [{tangentMode=RCTM_User, time=0, value=0}, <at 0.85 0.95>, {tangentMode=RCTM_User, time=1, value=1}]}}}
          CurveIndex=$Particles.PartitionParticles.Normalized Index In Partition
          Scale Curve=38
         )
        )
        4=&MakeVector(
         X=&FloatFromCurve(
          FloatCurve=DataInterface() {propertyValues {Curve {
             defaultValue=1
             keys [
              {interpMode=RCIM_Cubic, leaveTangentWeight=0.0692, tangentMode=RCTM_User, time=0, value=0}
              {arriveTangentWeight=0.0692, interpMode=RCIM_Cubic, leaveTangentWeight=0.0952, tangentMode=RCTM_User, time=0.2077, value=0}
              {arriveTangentWeight=0.0952, interpMode=RCIM_Cubic, leaveTangentWeight=0.1033, tangentMode=RCTM_User, time=0.5382, value=0.2696}
              {arriveTangentWeight=0.1033, interpMode=RCIM_Cubic, leaveTangentWeight=0.0656, tangentMode=RCTM_User, time=0.8033, value=0.3971}
              {arriveTangentWeight=0.0656, interpMode=RCIM_Cubic, tangentMode=RCTM_User, time=1, value=0}
             ]
            }}}
          CurveIndex=$Particles.PartitionParticles.Normalized Index In Partition
          Scale Curve=3
         )
         Y=&FloatFromCurve(
          FloatCurve=DataInterface() {propertyValues {Curve {
             defaultValue=1
             keys [
              {interpMode=RCIM_Cubic, leaveTangentWeight=0.0667, tangentMode=RCTM_User, time=0, value=0}
              {arriveTangentWeight=0.0667, interpMode=RCIM_Cubic, leaveTangentWeight=0.0988, tangentMode=RCTM_User, time=0.2, value=0}
              {arriveTangentWeight=0.0988, interpMode=RCIM_Cubic, leaveTangentWeight=0.1036, tangentMode=RCTM_User, time=0.5552, value=0.1471}
              {arriveTangentWeight=0.1036, interpMode=RCIM_Cubic, leaveTangentWeight=0.0642, tangentMode=RCTM_User, time=0.8239, value=-0.1569}
              {arriveTangentWeight=0.0642, interpMode=RCIM_Cubic, tangentMode=RCTM_User, time=1, value=0}
             ]
            }}}
          CurveIndex=$Particles.PartitionParticles.Normalized Index In Partition
          Scale Curve=3
         )
         Z=&FloatFromCurve(
          FloatCurve=DataInterface() {propertyValues {Curve {defaultValue=1, keys [{tangentMode=RCTM_User, time=0, value=0}, <at 0.85 0.95>, {tangentMode=RCTM_User, time=1, value=1}]}}}
          CurveIndex=$Particles.PartitionParticles.Normalized Index In Partition
          Scale Curve=40
         )
        )
        5=&MakeVector(
         X=&FloatFromCurve(
          FloatCurve=DataInterface() {propertyValues {Curve {
             defaultValue=1
             keys [
              {interpMode=RCIM_Cubic, leaveTangentWeight=0.0692, tangentMode=RCTM_User, time=0, value=0}
              {arriveTangentWeight=0.0692, interpMode=RCIM_Cubic, leaveTangentWeight=0.0952, tangentMode=RCTM_User, time=0.2077, value=0}
              {arriveTangentWeight=0.0952, interpMode=RCIM_Cubic, leaveTangentWeight=0.1033, tangentMode=RCTM_User, time=0.5982, value=-0.1373}
              {arriveTangentWeight=0.1033, interpMode=RCIM_Cubic, leaveTangentWeight=0.0656, tangentMode=RCTM_User, time=0.8084, value=0.1029}
              {arriveTangentWeight=0.0656, interpMode=RCIM_Cubic, tangentMode=RCTM_User, time=1, value=0}
             ]
            }}}
          CurveIndex=$Particles.PartitionParticles.Normalized Index In Partition
         )
         Y=&FloatFromCurve(
          FloatCurve=DataInterface() {propertyValues {Curve {
             defaultValue=1
             keys [
              {interpMode=RCIM_Cubic, leaveTangentWeight=0.0667, tangentMode=RCTM_User, time=0, value=0}
              {arriveTangentWeight=0.0667, interpMode=RCIM_Cubic, leaveTangentWeight=0.0988, tangentMode=RCTM_User, time=0.2, value=0}
              {arriveTangentWeight=0.0988, interpMode=RCIM_Cubic, leaveTangentWeight=0.1036, tangentMode=RCTM_User, time=0.5526, value=0.1225}
              {arriveTangentWeight=0.1036, interpMode=RCIM_Cubic, leaveTangentWeight=0.0642, tangentMode=RCTM_User, time=0.8098, value=-0.25}
              {arriveTangentWeight=0.0642, interpMode=RCIM_Cubic, tangentMode=RCTM_User, time=1, value=0}
             ]
            }}}
          CurveIndex=$Particles.PartitionParticles.Normalized Index In Partition
         )
         Z=&FloatFromCurve(
          FloatCurve=DataInterface() {propertyValues {Curve {defaultValue=1, keys [{tangentMode=RCTM_User, time=0, value=0}, <at 0.8 0.85>, {tangentMode=RCTM_User, time=1, value=1}]}}}
          CurveIndex=$Particles.PartitionParticles.Normalized Index In Partition
          Scale Curve=23
         )
        )
        Index=$Particles.PartitionParticles.Partition
        Num=::NewEnumerator4
       )
       B=&MakeCustomVectorFromBool(Bool=$User.IsLargeBody, True Vector=<1 1 0.75>, False Vector=<1 1 0.68>)
      )
     )
    )) {isSetParameters=true}
  ]
  rendererClasses [&NiagaraRibbonRendererProperties]
  renderers [{rendererClass=&NiagaraRibbonRendererProperties}]
 }
 V2Sparkles_FallingLeaves {
  emitterUpdateScript [
   &EmitterState()
   &SpawnBurst_Instantaneous(Spawn Count=15, Use Spawn Probability=false, UseLoopCountLimit=true)
   &SpawnRate(Spawn Probability=0.8, SpawnRate=25, Use Spawn Probability=true)
   &SpawnRate(Spawn Probability=0.75, SpawnGroup=1, SpawnRate=1.25, Use Spawn Probability=true) {module=SpawnRate001}
   &SpawnBurst_Instantaneous(Spawn Count=15, Spawn Time=3.7333, Use Spawn Probability=false, UseLoopCountLimit=false) {module=SpawnBurst_Instantaneous001}
  ]
  particleSpawnScript [
   /SetVariables(Particles.VisibilityTag=&SpawnGroup()) {isSetParameters=true}
   &ShapeLocation(
    Apply To Particle Position=::NewEnumerator1
    MaskSpawnGroup=false
    Override Seed=false
    Rotation Mode=::NewEnumerator5
    Shape Origin=&VectorFromFloat(Value=0)
    Sphere Radius=&MakeCustomFloatFromBool(Bool=$User.IsLargeBody, True Float=18, False Float=14)
    UseOwnerScale=true
   )
   &ShapeLocation(
    Apply To Particle Position=::NewEnumerator1
    Box Midpoint=<0.5 0.3 0.5>
    Box Size=&MakeCustomVectorFromBool(Bool=$User.IsLargeBody, True Vector=<70 10 25>, False Vector=<60 10 25>)
    MaskSpawnGroup=false
    Override Seed=false
    Rotation Mode=::NewEnumerator1
    Shape Origin=&VectorFromFloat(Value=0)
    Shape Primitive=::NewEnumerator2
    Surface Only=false
    UseOwnerScale=true
   ) {module=ShapeLocation001}
   &InitializeParticle(
    Color=&MergeHelper_ColorSwitch(
     0=&RandomRangeLinearColor(Minimum=<rgba 1 0.2912 0.4595 1>, Maximum=<rgba 1 0.3918 0.9784 1>)
     1=&ColorFromCurve(
      ColorCurve=DataInterface() {propertyValues {
        AlphaCurve {keys [<at 0 1 RCIM_Cubic>]}
        BlueCurve {keys [
          <at 0.0938 0.0021>
          {arriveTangent=0.041, interpMode=RCIM_Cubic, leaveTangent=0.041, time=0.1217, value=0.0097}
          <at 0.3586 0.013>
          <at 0.3901 0.2747>
          <at 0.6542 0.2747>
          <at 1 1 RCIM_Cubic>
         ]}
        GreenCurve {keys [
          <at 0.0938 0.0545>
          {arriveTangent=-0.2057, interpMode=RCIM_Cubic, leaveTangent=-0.2057, time=0.1217, value=0}
          <at 0.3586 0>
          <at 0.3901 0.0006>
          <at 0.6542 0.0006>
          <at 1 0 RCIM_Cubic>
         ]}
        RedCurve {keys [<at 0.0938 1>,<at 0.1217 1 RCIM_Cubic>,<at 0.3586 1>,<at 0.3901 1>,<at 0.6542 1>,<at 1 1 RCIM_Cubic>]}
       }}
      CurveIndex=&RandomRangeFloat(Minimum=0, Maximum=1)
     )
     2=?
     3=?
     Index=&SpawnGroup()
     Num=::NewEnumerator0
    )
    Lifetime Max=&MergeHelper_FloatSwitch(0=1.3, 1=1.2, 2=?, 3=?, Index=&SpawnGroup(), Num=::NewEnumerator0)
    Lifetime Min:NiagaraFloat=&MergeHelper_FloatSwitch(0=0.8, 1=0.8, 2=?, 3=?, Index=&SpawnGroup(), Num=::NewEnumerator0)
    Lifetime Mode=::NewEnumerator1
    Sprite Rotation Angle=&MergeHelper_FloatSwitch(
     0=0
     1=&RandomRangeFloat(Minimum=0, Maximum=360)
     2=?
     3=?
     Index=&SpawnGroup()
     Num=::NewEnumerator0
    )
    Sprite Rotation Mode=::NewEnumerator0
    Sprite Size Mode=::NewEnumerator1
    Sprite UV Mode=::NewEnumerator3
    Uniform Sprite Size Max=&MergeHelper_FloatSwitch(0=4, 1=9, 2=?, 3=?, Index=&SpawnGroup(), Num=::NewEnumerator0)
    Uniform Sprite Size Min=&MergeHelper_FloatSwitch(0=2, 1=7, 2=?, 3=?, Index=&SpawnGroup(), Num=::NewEnumerator0)
    UsePositionOffset=false
    Write Material Random=true
   )
   /SetVariables(Particles.RandomCore:NiagaraFloat=&FloatFromCurve(
     FloatCurve=DataInterface() {propertyValues {Curve {keys [
         {arriveTangent=0.0316, interpMode=RCIM_Cubic, leaveTangent=0.0316, tangentMode=RCTM_User, time=0, value=0.2}
         {arriveTangent=0.1895, interpMode=RCIM_Cubic, leaveTangent=0.1895, tangentMode=RCTM_User, time=0.9, value=0.25}
         {arriveTangent=1.1808, interpMode=RCIM_Cubic, leaveTangent=1.1808, tangentMode=RCTM_User, time=1, value=0.3}
        ]}}}
     CurveIndex=&RandomRangeFloat(Minimum=0, Maximum=1)
     Scale Curve=2
    )) {isSetParameters=true}
   &SkeletalMeshLocation(
    Bone / Socket Index=&MergeHelper_IntSwitch(
     0=&SelectIntFromArray(
      Array Sampling Mode=::NewEnumerator1
      Direct Array Index=?
      Int Selection Array=DataInterface() {propertyValues {IntData [1,2]}}
      Random Seed=0
     )
     1=7
     2=?
     3=?
     4=?
     5=?
     Index=&SpawnGroup()
     Num=::NewEnumerator0
    )
    Bone Sampling Mode=::NewEnumerator4
    "Inherit Velocity (Scale)":NiagaraFloat=&MergeHelper_FloatSwitch(0=-0.1, 1=-1, 2=?, 3=?, Index=&SpawnGroup(), Num=::NewEnumerator0)
    Inherited Velocity Speed Limit:NiagaraFloat=&MergeHelper_FloatSwitch(0=75, 1=25, 2=?, 3=?, Index=&SpawnGroup(), Num=::NewEnumerator0)
    MaskSpawnGroup=false
    Orientation Sampling=::NewEnumerator1
    Sampled Position Offset=&MergeHelper_VectorSwitch(
     0=&AddVectorToPosition(Position=$Particles.ShapeLocation.ShapePosition, Vector=<0 0 0>)
     1=&AddVectorToPosition(Position=$Particles.ShapeLocation001.ShapePosition, Vector=<15 0 0>)
     2=?
     3=?
     Index=&SpawnGroup()
     Num=::NewEnumerator0
    )
    Skeletal Mesh=$System.Skeletal Mesh
    Velocity Sampling=::NewEnumerator0
    VelocitySpeedLimit=true
   )
   &CameraOffset(Camera Offset Amount=&MergeHelper_FloatSwitch(
     0=&UniformAOrBFloat(A=15, B=-15, Distribution Weight=1, Evaluation Type=::NewEnumerator0, Recalculate Random Each Loop=true, Override Seed=false)
     1=-15
     2=?
     3=?
     Index=&SpawnGroup()
     Num=::NewEnumerator0
    ))
   &AddVelocity(
    Velocity=<0 0 1>
    Velocity Speed Scale=&MergeHelper_FloatSwitch(
     0=&FloatFromCurve(
      FloatCurve=DataInterface() {propertyValues {Curve {keys [<at 0 8>,<at 0.85 20>,<at 0.95 37>]}}}
      CurveIndex=&RandomRangeFloat(Minimum=0, Maximum=1)
      Scale Curve=&MakeCustomFloatFromBool(Bool=$User.IsLargeBody, True Float=1.1, False Float=1)
     )
     1=&RandomRangeFloat(Minimum=0, Maximum=50)
     2=?
     3=?
     Index=&SpawnGroup()
     Num=::NewEnumerator0
    )
   ) {module=AddVelocity001}
  ]
  particleUpdateScript [
   &ParticleState()
   /SetVariables(Particles.FlickerWaveform:NiagaraFloat=&Waveform(
     Global Phase=&Add_Float(A=$Engine.Time, B=&RandomRangeFloat(Minimum=0, Maximum=1))
     "Waveform [1]"=::NewEnumerator10
     "[1] Frequency"=10
     "[1] Amplitude Min/Max"=<0.5 1>
    )) {isSetParameters=true}
   &SubUVAnimation(
    Play Rate:NiagaraFloat=&RandomRangeFloat(Minimum=0.5, Maximum=1)
    Playback Scale Curve=DataInterface() {propertyValues {Curve {keys [
        {interpMode=RCIM_Cubic, tangentMode=RCTM_User, time=0.25, value=1}
        {arriveTangent=-1.2656, interpMode=RCIM_Cubic, leaveTangent=-1.2656, tangentMode=RCTM_User, time=1, value=0}
       ]}}}
    Random Start Frame=true
    "Sprite Renderer "=DataInterface() {propertyValues {SpriteRenderer=&self:V2Sparkles_FallingLeaves_0.NiagaraSpriteRendererProperties_3}}
    SubUV Animation Mode:ENiagara_SubUVLookupModeV2=::NewEnumerator3
    UseEndFrame=false
    UsePlayRateScale=false
    UseStartFrame=false
   )
   /SetVariables(Particles.LeafDynamicParameter0:Vector4f=&MakeVector4(X=0.1, Y=0, Z=0, W=0)) {isSetParameters=true}
   &ScaleSpriteSize(
    Scale Sprite Size Mode=::NewEnumerator0
    Uniform Scale Factor=&MergeHelper_FloatSwitch(
     0=&MakeFloatFromVector2d(Vector 2D=&Multiply_Vector2D_ByFloat(
       Float=1
       Vector2D=&Vector2DFromFloat(Value=&FloatFromCurve(
         FloatCurve=DataInterface() {propertyValues {Curve {keys [
             {arriveTangent=7.4528, interpMode=RCIM_Cubic, leaveTangent=7.4528, tangentMode=RCTM_User, time=0, value=0}
             <at 0.2 1.2 RCIM_Cubic>
             {arriveTangent=-2.7132, interpMode=RCIM_Cubic, leaveTangent=-2.7132, tangentMode=RCTM_User, time=1, value=0}
            ]}}}
         Scale Curve=&Multiply_Float(A=&FN_FrontEndFloatModifier(Gameplay=0.4, Front End=0.3), B=$Particles.FlickerWaveform)
        ))
      ))
     1=&FloatFromCurve(FloatCurve=DataInterface() {propertyValues {Curve {keys [
          {arriveTangent=14.5835, interpMode=RCIM_Cubic, leaveTangent=14.5835, tangentMode=RCTM_User, time=0, value=0}
          {arriveTangent=16.2546, interpMode=RCIM_Cubic, leaveTangent=0.0003, tangentMode=RCTM_Break, time=0.015, value=1}
          {arriveTangent=-0.2714, interpMode=RCIM_Cubic, leaveTangent=-3.4346, tangentMode=RCTM_Break, time=0.9, value=1}
          {arriveTangent=-20.5076, interpMode=RCIM_Cubic, leaveTangent=-20.5076, tangentMode=RCTM_User, time=1, value=0}
         ]}}})
     2=?
     3=?
     Index=&SpawnGroup()
     Num=::NewEnumerator0
    )
   )
   &ScaleColor(
    Scale Alpha=&MergeHelper_FloatSwitch(
     0=&MakeCustomFloatFromBool(
      Bool=$User.IsFrontEnd
      True Float=1
      False Float=/Niagara/DynamicInputs/Camera/CameraFade(
       Camera Query:NiagaraDataInterfaceCamera=DataInterface()
       Fade Offset:NiagaraFloat=100
       Fade Distance:NiagaraFloat=50
       Use Owner Position:NiagaraBool=true
      )
     )
     1=&Multiply_Float(
      A=&FloatFromCurve(FloatCurve=DataInterface() {propertyValues {Curve {keys [
           <at 0.1 1>
           {arriveTangent=-1.1111, interpMode=RCIM_Cubic, leaveTangent=-1.1111, time=0.8, value=1}
           {arriveTangent=-16.0205, interpMode=RCIM_Cubic, leaveTangent=-16.0205, tangentMode=RCTM_User, time=1, value=0}
          ]}}})
      B=&CameraFade(
       Camera Query:NiagaraDataInterfaceCamera=DataInterface()
       Fade Offset:NiagaraFloat=75
       Fade Distance:NiagaraFloat=5
       Use Owner Position:NiagaraBool=true
      )
     )
     2=?
     3=?
     Index=&SpawnGroup()
     Num=::NewEnumerator0
    )
    Scale RGB=&MergeHelper_VectorSwitch(
     0=&Multiply_VectorByFloat(
      Vector=&MakeVectorFromLinearColorRGB(Color=&ColorFromCurve(
        ColorCurve=$System.Color Curve
        CurveIndex=&Frac_Float(Float=&Add_Float(A=&RandomRangeFloat(Minimum=0, Maximum=1), B=$Particles.NormalizedAge))
       ))
      Float=&FloatFromCurve(
       FloatCurve=DataInterface() {propertyValues {Curve {
          defaultValue=1
          keys [
           {interpMode=RCIM_Cubic, tangentMode=RCTM_User, time=0.5, value=1}
           {arriveTangent=-3.9224, interpMode=RCIM_Cubic, leaveTangent=-3.9224, tangentMode=RCTM_User, time=1, value=0.2}
          ]
         }}}
       Scale Curve=&Multiply_Float(
        A=15
        B=&FN_Fit(
         Clamp:NiagaraBool=false
         Value:NiagaraFloat=$Particles.FlickerWaveform
         SourceMin:NiagaraFloat=0.5
         SourceMax:NiagaraFloat=1
         DestMin:NiagaraFloat=0.9
         DestMax:NiagaraFloat=1.2
        )
       )
      )
     )
     1=<1 1 1>
     2=?
     3=?
     Index=&SpawnGroup()
     Num=::NewEnumerator0
    )
    ScaleA=true
    ScaleRGB=true
   )
   /SetVariables(Particles.Glow03_Size=&Multiply_Vector2D_ByFloat(
     Float=&FloatFromCurve(
      FloatCurve=DataInterface() {propertyValues {Curve {keys [
          <at 0 0 RCIM_Cubic>
          <at 0.2 1.3 RCIM_Cubic>
          {arriveTangent=-0.1423, interpMode=RCIM_Cubic, leaveTangent=-0.1423, tangentMode=RCTM_User, time=1, value=0}
         ]}}}
      Scale Curve=&Multiply_Float(A=&MakeFloatFromVector2d(Vector 2D=$Particles.SpriteSize), B=3)
     )
     Vector2D=&RandomRangeVector2D(Minimum=<1 2.5>, Maximum=<1 5>)
    )) {isSetParameters=true}
   /SetVariables(Particles.Glow03_Color=&MakeLinearColorFromVectorAndFloat(
     "Vector (RGB)"=&MakeVectorFromLinearColorRGB(Color=&Multiply_LinearColorByFloat(
       Linear Color=&MultiplyLinearColor(Linear Color=$Particles.Color, Scale Factor=&MakeVector4_FromLinearColor(LinearColor=$Particles.Color))
       Float=0.07
      ))
     "Float (Alpha)"=&Multiply_Float(
      A=&Multiply_Float(
       A=0.025
       B=&CameraFade(
        Camera Query:NiagaraDataInterfaceCamera=DataInterface()
        Fade Offset:NiagaraFloat=100
        Fade Distance:NiagaraFloat=5
        Use Owner Position:NiagaraBool=true
       )
      )
      B=&MakeCustomFloatFromBool(Bool=$User.IsFrontEnd, True Float=0.3, False Float=1)
     )
    )) {isSetParameters=true}
   &CurlNoiseForce(
    Bias Noise=false
    Noise Frequency=200
    Noise Strength=&MergeHelper_FloatSwitch(
     0=&Multiply_Float(
      A=&FloatFromCurve(
       FloatCurve=DataInterface() {propertyValues {Curve {
          defaultValue=1
          keys [
           {interpMode=RCIM_Cubic, leaveTangentWeight=0.5765, tangentMode=RCTM_User, time=0, value=0}
           {interpMode=RCIM_Cubic, tangentMode=RCTM_User, time=0.8, value=1}
          ]
         }}}
       Scale Curve=25
      )
      B=1
     )
     1=0
     2=?
     3=?
     Index=&SpawnGroup()
     Num=::NewEnumerator0
    )
    Pan Noise=true
    Pan Noise Field=<0 0 1>
   ) {enabled=false}
   &GravityForce(Gravity=&Multiply_VectorByFloat(
     Vector=<0 0 -1>
     Float=&MergeHelper_FloatSwitch(
      0=0
      1=&RandomRangeFloat(Minimum=175, Maximum=300)
      2=?
      3=?
      Index=&SpawnGroup()
      Num=::NewEnumerator0
     )
    ))
   &Drag(
    Drag=&MergeHelper_FloatSwitch(
     0=1
     1=&RandomRangeFloat(Minimum=2, Maximum=10)
     2=?
     3=?
     Index=&SpawnGroup()
     Num=::NewEnumerator0
    )
    Use Linear Drag=true
    Use Rotational Drag=true
   )
   &SolveForcesAndVelocity(Clamp Velocity=false, Limit Acceleration=false)
  ]
  rendererClasses [&NiagaraSpriteRendererProperties]
  renderers [
   {rendererClass=&NiagaraSpriteRendererProperties}
   {rendererClass=&NiagaraSpriteRendererProperties, rendererIndex=1}
   {rendererClass=&NiagaraSpriteRendererProperties, rendererIndex=2}
  ]
 }
}
summary {
 systemName=NS_Aura_FantasyPlant
 emitters [
  {emitterName=Stem_Ribbons, rendererClasses [&NiagaraRibbonRendererProperties]}
  {
   bEnabled=false
   emitterName=Static_Icons
   rendererClasses [&NiagaraSpriteRendererProperties, &NiagaraMeshRendererProperties]
  }
  {
   bEnabled=false
   emitterName=Spawn_Icons_attached
   rendererClasses [&NiagaraSpriteRendererProperties, &NiagaraMeshRendererProperties]
  }
  {emitterName=Decal_Glows_Combined, rendererClasses [&NiagaraSpriteRendererProperties]}
  {emitterName=V2Sparkles_FallingLeaves, rendererClasses [&NiagaraSpriteRendererProperties]}
  {emitterName=Leaf_Sprites001, rendererClasses [&NiagaraSpriteRendererProperties]}
  {
   bEnabled=false
   emitterName=DebugRotationLerp
   rendererClasses [&NiagaraSpriteRendererProperties, &NiagaraMeshRendererProperties]
  }
  {
   bEnabled=false
   emitterName=Spawn_Icons_attached001
   rendererClasses [&NiagaraSpriteRendererProperties, &NiagaraMeshRendererProperties]
  }
  {
   bEnabled=false
   emitterName=DebugRotationLerp002
   rendererClasses [&NiagaraSpriteRendererProperties, &NiagaraMeshRendererProperties]
  }
 ]
 userVariables [
  User.bIsFrontEnd {type=NiagaraBool}
  User.bIsFrontEndPreview {type=NiagaraBool}
  User.IsFrontEnd {type=NiagaraBool}
  User.IsLargeBody {type=NiagaraBool}
 ]
}
